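; Behavior tree service that puts the Caretaker NPC into a combat frustration state after the service has been active for a set time.
; On activation it sets a deadline from the scalable max frustration duration, falling back to a default when that value is not positive.
; Once game time passes the deadline, tick applies the CombatFrustration gameplay effect once to the controlled pawn.
; Deactivation removes active effects that grant the CombatFrustration.Active tag.
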(event EventReceiveActivation
  (params (OwnerActor :type Actor%20Object%20Reference))
  (bind _maxfrustrationduration (Variables|Default|GetMaxFrustrationDuration))
  (bind _returnvalue (ScalableFloat|GetValueatLevel _maxfrustrationduration))
  (Variables|Default|SetfrustrationApplied false)
  (if (> _returnvalue 0.0)
    (Variables|Default|SetFrustrationTime (+ (Utilities|Time|GetGameTimeinSeconds) _returnvalue))
    (else
      (Variables|Default|SetFrustrationTime (+ (Utilities|Time|GetGameTimeinSeconds) (Variables|Default|GetDefaultFrustrationDuration))))))

(event EventReceiveTick
  (params (OwnerActor :type Actor%20Object%20Reference) (DeltaSeconds :type Float%20%28single-precision%29))
  (if (and (>= (Utilities|Time|GetGameTimeinSeconds) (Variables|Default|GetFrustrationTime)) (not (Variables|Default|GetfrustrationApplied)))
    (Variables|Default|SetfrustrationApplied true)
    (bind _asaicontroller (Utilities|Casting|CastToAIController OwnerActor))
    (Utilities|IsValid (Pawn|GetControlledPawn _asaicontroller)
      (:Is%20Valid
        (Utilities|IsValid (Ability|GetAbilitySystemComponent (Pawn|GetControlledPawn _asaicontroller))
          (:Is%20Valid
            (GameplayEffects|ApplyGameplayEffectToSelf :self (Ability|GetAbilitySystemComponent (Pawn|GetControlledPawn _asaicontroller)) :GameplayEffectClass "/Caretaker/Abilities/GE_NPC_Irwin_Monster_Caretaker_CombatFrustration.GE_NPC_Irwin_Monster_Caretaker_CombatFrustration_C" :Level 0.0)))))))

(event EventReceiveDeactivation
  (params (OwnerActor :type Actor%20Object%20Reference))
  (bind _asaicontroller (Utilities|Casting|CastToAIController OwnerActor))
  (Utilities|IsValid (Pawn|GetControlledPawn _asaicontroller)
    (:Is%20Valid
      (Utilities|IsValid (Ability|GetAbilitySystemComponent (Pawn|GetControlledPawn _asaicontroller))
        (:Is%20Valid
          (GameplayEffects|RemoveActiveEffectswithGrantedTags (Ability|GetAbilitySystemComponent (Pawn|GetControlledPawn _asaicontroller)) "(GameplayTags=((TagName=\"NPC.Ability.Attack.Caretaker.CombatFrustration.Active\")))"))))))
