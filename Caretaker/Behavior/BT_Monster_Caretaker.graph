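# Behavior tree for the Caretaker monster, choosing attacks and movement against the GoalActor.
# When pathing fails and no attack is active, it rotates, moves directly if off the navmesh, or waits while frustrated.
# Combat prioritizes minion summon, primary melee, storm-escape wandering, then ranged swipe with nav-filtered approach, gated by ability range and activation checks.
# Looped turn-to-target steps yield to a combat frustration tag, and moving attacks stop and pause before executing.
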BehaviorTree /Caretaker/Behavior/BT_Monster_Caretaker
blackboard=&/FortAIBB_Caretaker
root {
 name=Selector
 class=BTComposite_Selector
 children [
  Selector {
   class=BTComposite_Selector
   children [
    Selector {
     class=BTComposite_Selector
     children [
      Sequence {
       class=BTComposite_Sequence
       children [
        Rotate to face BB entry {class=BTTask_RotateToFaceBBEntry, decorators [], properties {blackboardKey {selectedKeyName=GoalActor}, precision {defaultValue=10}}}
        Wait {class=BTTask_Wait, properties {waitTime {defaultValue=0.5}}}
       ]
       decorators [
        Not Escaping Into Storm {
         class=BTDecorator_GameplayTagQuery
         properties {
          flowAbortMode=None
          nodeName=Not Escaping Into Storm
          gameplayTagQuery {
           autoDescription=" NONE( NPC.Ability.NonAttack.Caretaker.EscapeIntoStorm )"
           queryTokenStream [0,1,3,1,0]
           tagDictionary [{tagName=NPC.Ability.NonAttack.Caretaker.EscapeIntoStorm}]
          }
         }
        }
        NOT Rotated To Attack Target {
         class=FortBTDecorator_GameplayAbility_IsRotatedToAttackTarget
         properties {
          bInverseCondition=true
          flowAbortMode=None
          nodeName=NOT Rotated To Attack Target
          activeAbilityTagsToSkipTestingValue {key=None, defaultValue {gameplayTags []}}
          gameplayAbilityTagContainer {key=None, defaultValue {gameplayTags [{tagName=NPC.Ability.Attack.Melee.Primary}]}}
          targetValue {baseClass=&Actor, defaultValue=null, key=GoalActor}
         }
        }
       ]
       properties {children [
         {childTask=&self:BTTask_RotateToFaceBBEntry_3}
         {childTask=&self:BTTask_Wait_2}
        ]}
      }
      CaretakerMoveTo Direct {
       class=FortBTTask_CaretakerMoveTo
       children []
       decorators [Off Nav Mesh {
         class=BTDecorator_Blackboard
         properties {
          arithmeticOperation=Equal
          basicOperation=NotSet
          flowAbortMode=None
          intValue=4
          nodeName=Off Nav Mesh
          stringValue=EBotMovementState::Failed_AgentOffNavmesh
          blackboardKey {selectedKeyName=AIEvaluator_PathExists}
         }
        }]
       properties {
        bEnableSlowdownAtGoal=false
        bFinishMoveOnOverlap=true
        bIgnoreRestartSelf=false
        bMoveDirectlyTowards=true
        bStopAtGoal=true
        nodeName=CaretakerMoveTo Direct
        pathObstacleAction=Ignore
        acceptableRadius {defaultValue=750, key=None}
        bAllowPartialPath {defaultValue=true, key=None}
        bAllowStrafe {defaultValue=false, key=None}
        bProjectGoalLocation {defaultValue=true, key=None}
        bReachTestIncludesAgentRadius {defaultValue=true, key=None}
        bReachTestIncludesGoalRadius {defaultValue=true, key=None}
        bRequireNavigableEndLocation {defaultValue=true, key=None}
        bStartFromPreviousPath {defaultValue=false, key=None}
        bTrackMovingGoal {defaultValue=true, key=None}
        blackboardKey {bNoneIsAllowedValue=false, selectedKeyName=GoalActor}
        filterClass {
         baseClass=&NavigationQueryFilter
         defaultValue=/Caretaker/NavFilter/NavFilter_Caretaker_NoBlockers.NavFilter_Caretaker_NoBlockers_C
         key=None
        }
        focalPointWhileMoving {bNoneIsAllowedValue=true, selectedKeyName=None}
        observedBlackboardValueTolerance {defaultValue=4.75, key=None}
       }
       services []
      }
      Wait To Teleport? {
       class=BTTask_Wait
       decorators [Can't Move? {
         class=BTDecorator_Blackboard
         properties {
          arithmeticOperation=Equal
          flowAbortMode=None
          intValue=0
          nodeName=Can't Move?
          stringValue=""
          blackboardKey {selectedKeyName=IsFrustrated}
         }
        }]
       properties {nodeName=Wait To Teleport?, waitTime {defaultValue=2}}
      }
     ]
     decorators [
      Has Active Ability with Any Matching Tag {
       class=HasActiveAbilityWithAnyMatchingTag_C
       properties {
        bCheckConditionOnlyBlackBoardChanges=false
        bInverseCondition=true
        bShowPropertyDetails=true
        customDescription=""
        flowAbortMode=None
        nodeName=Has Active Ability with Any Matching Tag
        activeAbilityTagsToCheck {gameplayTags [{tagName=NPC.Ability.Attack.Summon.Minions}]}
       }
      }
      Has Active Ability with Any Matching Tag {
       class=HasActiveAbilityWithAnyMatchingTag_C
       properties {
        bCheckConditionOnlyBlackBoardChanges=false
        bInverseCondition=true
        bShowPropertyDetails=true
        customDescription=""
        flowAbortMode=None
        nodeName=Has Active Ability with Any Matching Tag
        activeAbilityTagsToCheck {gameplayTags [{tagName=NPC.Ability.Attack.Ranged.Primary}]}
       }
      }
      Has Active Ability with Any Matching Tag {
       class=HasActiveAbilityWithAnyMatchingTag_C
       properties {
        bCheckConditionOnlyBlackBoardChanges=false
        bInverseCondition=true
        bShowPropertyDetails=true
        customDescription=""
        flowAbortMode=None
        nodeName=Has Active Ability with Any Matching Tag
        activeAbilityTagsToCheck {gameplayTags [{tagName=NPC.Ability.Attack.Melee.Primary}]}
       }
      }
      Has Active Ability with Any Matching Tag {
       class=HasActiveAbilityWithAnyMatchingTag_C
       properties {
        bCheckConditionOnlyBlackBoardChanges=false
        bInverseCondition=true
        bShowPropertyDetails=true
        customDescription=""
        flowAbortMode=None
        nodeName=Has Active Ability with Any Matching Tag
        activeAbilityTagsToCheck {gameplayTags [{tagName=NPC.Ability.Attack.Melee.Secondary}]}
       }
      }
      Blackboard Based Condition {
       class=BTDecorator_Blackboard
       properties {
        arithmeticOperation=NotEqual
        intValue=9
        nodeName=Blackboard Based Condition
        stringValue=EBotMovementState::Success
        blackboardKey {selectedKeyName=AIEvaluator_PathExists}
       }
      }
      Blackboard Based Condition {
       class=BTDecorator_Blackboard
       properties {
        arithmeticOperation=NotEqual
        intValue=10
        nodeName=Blackboard Based Condition
        stringValue=EBotMovementState::Success_Partial
        blackboardKey {selectedKeyName=AIEvaluator_PathExists}
       }
      }
     ]
     properties {
      children [
       {
        childComposite=&self:BTComposite_Sequence_1
        childTask=null
        decorators [
         &self:BTDecorator_GameplayTagQuery_5
         &self:FortBTDecorator_GameplayAbility_IsRotatedToAttackTarget_1
        ]
       }
       {
        childComposite=null
        childTask=&self:FortBTTask_CaretakerMoveTo_0
        decorators [&self:BTDecorator_Blackboard_1]
       }
       {
        childComposite=null
        childTask=&self:BTTask_Wait_1
        decorators [&self:BTDecorator_Blackboard_2]
       }
      ]
      services []
     }
     services []
    }
    Selector - Combat {
     class=BTComposite_Selector
     children [
      Selector {
       class=BTComposite_Selector
       children [
        Sequence {
         class=BTComposite_Sequence
         children [
          StopMovement {
           class=BTTask_StopMovement_C
           children []
           decorators []
           properties {bIgnoreRestartSelf=false, bShowPropertyDetails=true, customDescription="", nodeName=StopMovement, tickInterval {interval=-1}}
           services []
          }
          Wait {class=BTTask_Wait, properties {waitTime {defaultValue=2.5}}}
          GameplayAbility {
           class=FortBTTask_ExecuteGameplayAbility
           decorators []
           properties {nodeName=GameplayAbility, gameplayAbilityTagContainer {defaultValue {gameplayTags [{tagName=NPC.Ability.Attack.Summon.Minions}]}}}
          }
         ]
         decorators [IsMoving {
           class=FortBTDecorator_IsMoving
           properties {
            bInverseCondition=false
            bUseMinDist=true
            flowAbortMode=LowerPriority
            nodeName=IsMoving
            activeAbilityTagsToSkipTestingValue {key=None, defaultValue {gameplayTags []}}
            gameplayAbilityTagContainer {key=None, defaultValue {gameplayTags [{tagName=NPC.Ability.Attack.Summon.Minions}]}}
            minDistanceComparison {
             addedOffset=0
             bUseAddedOffset=false
             bUseOverriddenValue=false
             comparisonType=ThreeDimensions
             operator=Less
             overriddenValue=0
             distanceDataTags {gameplayTags [{tagName=NPC.Ability.MinStopRange}]}
            }
            minTimeValue {defaultValue=2, key=None}
            targetValue {baseClass=&Actor, defaultValue=null, key=GoalActor}
            updateIntervalValue {defaultValue=0.1, key=None}
            useMinDistToTargetTime {defaultValue=1, key=None}
           }
          }]
         properties {children [
           {childTask=&self:BTTask_StopMovement_C_4}
           {childTask=&self:BTTask_Wait_18}
           {childTask=&self:FortBTTask_ExecuteGameplayAbility_1}
          ]}
        }
        Sequence {
         class=BTComposite_Sequence
         children [GameplayAbility {
           class=FortBTTask_ExecuteGameplayAbility
           decorators []
           properties {nodeName=GameplayAbility, gameplayAbilityTagContainer {defaultValue {gameplayTags [{tagName=NPC.Ability.Attack.Summon.Minions}]}}}
          }]
         decorators [IsNotMoving {
           class=FortBTDecorator_IsMoving
           properties {
            bInverseCondition=true
            bUseMinDist=false
            flowAbortMode=LowerPriority
            nodeName=IsNotMoving
            activeAbilityTagsToSkipTestingValue {key=None, defaultValue {gameplayTags []}}
            gameplayAbilityTagContainer {key=None, defaultValue {gameplayTags []}}
            minDistanceComparison {
             addedOffset=0
             bUseAddedOffset=false
             bUseOverriddenValue=false
             comparisonType=TwoDimensions
             operator=Equal
             overriddenValue=0
             distanceDataTags {gameplayTags []}
            }
            minTimeValue {defaultValue=0, key=None}
            targetValue {baseClass=&Actor, defaultValue=null, key=GoalActor}
            updateIntervalValue {defaultValue=0.1, key=None}
            useMinDistToTargetTime {defaultValue=0, key=None}
           }
          }]
         properties {children [{childTask=&self:FortBTTask_ExecuteGameplayAbility_18}]}
        }
       ]
       decorators [
        "Has Gameplay Ability - Summon Minions (Cast Sideways)" {
         class=FortBTDecorator_GameplayAbility_HasGameplayAbility
         properties {
          nodeName="Has Gameplay Ability - Summon Minions (Cast Sideways)"
          gameplayAbilityTagContainer {defaultValue {gameplayTags [{tagName=NPC.Ability.Attack.Summon.Minions}]}}
         }
        }
        Can Activate Gameplay Ability {
         class=FortBTDecorator_GameplayAbility_CanActivate
         properties {
          flowAbortMode=None
          gameplayAbilityTagContainer {defaultValue {gameplayTags [{tagName=NPC.Ability.Attack.Summon.Minions}]}}
          targetValue {key=GoalActor}
         }
        }
        Within Cast Sideways Actual Range? {
         class=FortBTDecorator_GameplayAbility_CompareDistance
         properties {
          bInverseCondition=false
          flowAbortMode=LowerPriority
          nodeName=Within Cast Sideways Actual Range?
          activeAbilityTagsToSkipTestingValue {key=None, defaultValue {gameplayTags []}}
          distanceComparisons [{
            addedOffset=0
            bUseAddedOffset=false
            bUseOverriddenValue=false
            comparisonType=ThreeDimensions
            operator=Less
            overriddenValue=0
            distanceDataTags {gameplayTags [{tagName=NPC.Ability.MaxActualRange}]}
           }]
          gameplayAbilityTagContainer {key=None, defaultValue {gameplayTags [{tagName=NPC.Ability.Attack.Summon.Minions}]}}
          targetValue {baseClass=&Actor, defaultValue=null, key=GoalActor}
         }
        }
       ]
       properties {
        children [
         {
          childComposite=&self:BTComposite_Sequence_13
          childTask=null
          decorators [&self:FortBTDecorator_IsMoving_7]
         }
         {
          childComposite=&self:BTComposite_Sequence_15
          childTask=null
          decorators [&self:FortBTDecorator_IsMoving_2]
         }
        ]
        services []
       }
       services []
      }
      Selector {
       class=BTComposite_Selector
       children [
        Selector {
         class=BTComposite_Selector
         children [Sequence {
           class=BTComposite_Sequence
           children [
            Rotate to face BB entry {class=BTTask_RotateToFaceBBEntry, decorators [], properties {blackboardKey {selectedKeyName=GoalActor}, precision {defaultValue=10}}}
            Wait {class=BTTask_Wait, properties {waitTime {defaultValue=0.5}}}
           ]
           decorators [
            Within Primary Melee Approach Range? {
             class=FortBTDecorator_GameplayAbility_CompareDistance
             properties {
              bInverseCondition=false
              flowAbortMode=Self
              nodeName=Within Primary Melee Approach Range?
              activeAbilityTagsToSkipTestingValue {key=None, defaultValue {gameplayTags []}}
              distanceComparisons [{
                addedOffset=0
                bUseAddedOffset=false
                bUseOverriddenValue=false
                comparisonType=ThreeDimensions
                operator=Less
                overriddenValue=0
                distanceDataTags {gameplayTags [{tagName=NPC.Ability.MaxActualRange}]}
               }]
              gameplayAbilityTagContainer {key=None, defaultValue {gameplayTags [{tagName=NPC.Ability.Attack.Melee.Primary}]}}
              targetValue {baseClass=&Actor, defaultValue=null, key=GoalActor}
             }
            }
            Could Hit if we Rotated {
             class=FortBTDecorator_GameplayAbility_CanHitTarget
             properties {
              bInverseCondition=false
              flowAbortMode=None
              nodeName=Could Hit if we Rotated
              activeAbilityTagsToSkipTestingValue {key=None, defaultValue {gameplayTags []}}
              gameplayAbilityTagContainer {key=None, defaultValue {gameplayTags [{tagName=NPC.Ability.Attack.Melee.Primary}]}}
              targetValue {baseClass=&Actor, defaultValue=null, key=GoalActor}
              useIdealYawRotationToTargetValue {defaultValue=true, key=None}
             }
            }
            NOT Rotated To Attack Target {
             class=FortBTDecorator_GameplayAbility_IsRotatedToAttackTarget
             properties {
              bInverseCondition=true
              flowAbortMode=None
              nodeName=NOT Rotated To Attack Target
              activeAbilityTagsToSkipTestingValue {key=None, defaultValue {gameplayTags []}}
              gameplayAbilityTagContainer {key=None, defaultValue {gameplayTags [{tagName=NPC.Ability.Attack.Melee.Primary}]}}
              targetValue {baseClass=&Actor, defaultValue=null, key=GoalActor}
             }
            }
            Is not frustrated {
             class=BTDecorator_GameplayTagQuery
             properties {
              flowAbortMode=Self
              nodeName=Is not frustrated
              gameplayTagQuery {
               autoDescription=" NONE( NPC.Ability.Attack.Caretaker.CombatFrustration.Active )"
               queryTokenStream [0,1,3,1,0]
               tagDictionary [{tagName=NPC.Ability.Attack.Caretaker.CombatFrustration.Active}]
              }
             }
            }
           ]
           properties {
            children [
             {childTask=&self:BTTask_RotateToFaceBBEntry_4}
             {childTask=&self:BTTask_Wait_5}
            ]
            services [&self:BTS_CaretakerCombatFrustration_C_0]
           }
           services [BTS_CaretakerCombatFrustration_C_0 {
             class=BTS_CaretakerCombatFrustration_C
             properties {
              bCallTickOnSearchStart=false
              bRestartTimerOnEachActivation=false
              bShowEventDetails=false
              bShowPropertyDetails=true
              customDescription=""
              defaultFrustrationDuration=5
              frustrationApplied=false
              frustrationTime=0
              interval=0.5
              nodeName=""
              randomDeviation=0.1
              maxFrustrationDuration {
               registryType=None
               value=1
               curve {curveTable=/Caretaker/DataTables/NPCMMRScaling_Caretaker, rowName=Caretaker.Attack.General.CombatFrustrationDuration}
              }
             }
            }]
          }]
         decorators [Loop {class=BTDecorator_Loop}]
         properties {
          children [{
            childComposite=&self:BTComposite_Sequence_6
            childTask=null
            decorators [
             &self:FortBTDecorator_GameplayAbility_CompareDistance_6
             &self:FortBTDecorator_GameplayAbility_CanHitTarget_2
             &self:FortBTDecorator_GameplayAbility_IsRotatedToAttackTarget_2
             &self:BTDecorator_GameplayTagQuery_16
            ]
           }]
          services []
         }
         services []
        }
        Sequence {
         class=BTComposite_Sequence
         children [
          StopMovement {
           class=BTTask_StopMovement_C
           children []
           decorators []
           properties {bIgnoreRestartSelf=false, bShowPropertyDetails=true, customDescription="", nodeName=StopMovement, tickInterval {interval=-1}}
           services []
          }
          Wait {class=BTTask_Wait, properties {waitTime {defaultValue=2.5}}}
          GameplayAbility {
           class=FortBTTask_ExecuteGameplayAbility
           decorators []
           properties {nodeName=GameplayAbility, gameplayAbilityTagContainer {defaultValue {gameplayTags [{tagName=NPC.Ability.Attack.Melee.Primary}]}}}
          }
         ]
         decorators [IsMoving {
           class=FortBTDecorator_IsMoving
           properties {
            bInverseCondition=false
            bUseMinDist=true
            flowAbortMode=LowerPriority
            nodeName=IsMoving
            activeAbilityTagsToSkipTestingValue {key=None, defaultValue {gameplayTags []}}
            gameplayAbilityTagContainer {key=None, defaultValue {gameplayTags [{tagName=NPC.Ability.Attack.Melee.Primary}]}}
            minDistanceComparison {
             addedOffset=0
             bUseAddedOffset=false
             bUseOverriddenValue=false
             comparisonType=ThreeDimensions
             operator=Less
             overriddenValue=0
             distanceDataTags {gameplayTags [{tagName=NPC.Ability.MinStopRange}]}
            }
            minTimeValue {defaultValue=2, key=None}
            targetValue {baseClass=&Actor, defaultValue=null, key=GoalActor}
            updateIntervalValue {defaultValue=0.1, key=None}
            useMinDistToTargetTime {defaultValue=1, key=None}
           }
          }]
         properties {children [
           {childTask=&self:BTTask_StopMovement_C_3}
           {childTask=&self:BTTask_Wait_17}
           {childTask=&self:FortBTTask_ExecuteGameplayAbility_2}
          ]}
        }
        Sequence {
         class=BTComposite_Sequence
         children [GameplayAbility {
           class=FortBTTask_ExecuteGameplayAbility
           decorators []
           properties {nodeName=GameplayAbility, gameplayAbilityTagContainer {defaultValue {gameplayTags [{tagName=NPC.Ability.Attack.Melee.Primary}]}}}
          }]
         decorators [IsNotMoving {
           class=FortBTDecorator_IsMoving
           properties {
            bInverseCondition=true
            bUseMinDist=false
            flowAbortMode=LowerPriority
            nodeName=IsNotMoving
            activeAbilityTagsToSkipTestingValue {key=None, defaultValue {gameplayTags []}}
            gameplayAbilityTagContainer {key=None, defaultValue {gameplayTags []}}
            minDistanceComparison {
             addedOffset=0
             bUseAddedOffset=false
             bUseOverriddenValue=false
             comparisonType=TwoDimensions
             operator=Equal
             overriddenValue=0
             distanceDataTags {gameplayTags []}
            }
            minTimeValue {defaultValue=0, key=None}
            targetValue {baseClass=&Actor, defaultValue=null, key=GoalActor}
            updateIntervalValue {defaultValue=0.1, key=None}
            useMinDistToTargetTime {defaultValue=0, key=None}
           }
          }]
         properties {children [{childTask=&self:FortBTTask_ExecuteGameplayAbility_5}]}
        }
       ]
       decorators [
        Has Gameplay Ability - Melee Primary {
         class=FortBTDecorator_GameplayAbility_HasGameplayAbility
         properties {
          nodeName=Has Gameplay Ability - Melee Primary
          gameplayAbilityTagContainer {defaultValue {gameplayTags [{tagName=NPC.Ability.Attack.Melee.Primary}]}}
         }
        }
        Can Activate Gameplay Ability {
         class=FortBTDecorator_GameplayAbility_CanActivate
         properties {
          flowAbortMode=None
          gameplayAbilityTagContainer {defaultValue {gameplayTags [{tagName=NPC.Ability.Attack.Melee.Primary}]}}
          targetValue {key=GoalActor}
         }
        }
        Within Primary Melee AoE Max Actual Range? {
         class=FortBTDecorator_GameplayAbility_CompareDistance
         properties {
          bInverseCondition=false
          flowAbortMode=LowerPriority
          nodeName=Within Primary Melee AoE Max Actual Range?
          activeAbilityTagsToSkipTestingValue {key=None, defaultValue {gameplayTags []}}
          distanceComparisons [{
            addedOffset=0
            bUseAddedOffset=false
            bUseOverriddenValue=false
            comparisonType=ThreeDimensions
            operator=Less
            overriddenValue=0
            distanceDataTags {gameplayTags [{tagName=NPC.Ability.MaxActualRange}]}
           }]
          gameplayAbilityTagContainer {key=None, defaultValue {gameplayTags [{tagName=NPC.Ability.Attack.Melee.Primary}]}}
          targetValue {baseClass=&Actor, defaultValue=null, key=GoalActor}
         }
        }
       ]
       properties {
        children [
         {
          childComposite=&self:BTComposite_Selector_13
          childTask=null
          decorators [&self:BTDecorator_Loop_1]
         }
         {
          childComposite=&self:BTComposite_Sequence_4
          childTask=null
          decorators [&self:FortBTDecorator_IsMoving_6]
         }
         {
          childComposite=&self:BTComposite_Sequence_14
          childTask=null
          decorators [&self:FortBTDecorator_IsMoving_5]
         }
        ]
        services []
       }
       services []
      }
      Selector {
       class=BTComposite_Selector
       children [
        GameplayAbility {
         class=FortBTTask_ExecuteGameplayAbility
         decorators [
          Can Activate Gameplay Ability {
           class=FortBTDecorator_GameplayAbility_CanActivate
           properties {
            flowAbortMode=None
            gameplayAbilityTagContainer {defaultValue {gameplayTags [{tagName=NPC.Ability.NonAttack.Wander.Escape}]}}
            targetValue {key=GoalActor}
           }
          }
          Has NO GoalActor? {
           class=BTDecorator_Blackboard
           properties {
            arithmeticOperation=Equal
            basicOperation=NotSet
            intValue=0
            nodeName=Has NO GoalActor?
            stringValue=""
            blackboardKey {selectedKeyName=GoalActor}
           }
          }
         ]
         properties {nodeName=GameplayAbility, gameplayAbilityTagContainer {defaultValue {gameplayTags [{tagName=NPC.Ability.NonAttack.Wander.Escape}]}}}
        }
        GameplayAbility {
         class=FortBTTask_ExecuteGameplayAbility
         decorators [
          Can Activate Gameplay Ability {
           class=FortBTDecorator_GameplayAbility_CanActivate
           properties {
            flowAbortMode=None
            gameplayAbilityTagContainer {defaultValue {gameplayTags [{tagName=NPC.Ability.NonAttack.Wander.Escape}]}}
            targetValue {key=GoalActor}
           }
          }
          Has GoalActor? {
           class=BTDecorator_Blackboard
           properties {
            arithmeticOperation=Equal
            intValue=0
            nodeName=Has GoalActor?
            notifyObserver=ValueChange
            stringValue=""
            blackboardKey {selectedKeyName=GoalActor}
           }
          }
         ]
         properties {nodeName=GameplayAbility, gameplayAbilityTagContainer {defaultValue {gameplayTags [{tagName=NPC.Ability.NonAttack.Wander.Escape}]}}}
        }
        Selector {
         class=BTComposite_Selector
         children [Wait {class=BTTask_Wait, properties {waitTime {defaultValue=1}}}]
         decorators [Is On Cooldown {
           class=FortBTDecorator_GameplayAbility_IsOnCooldown
           properties {
            bInverseCondition=false
            flowAbortMode=None
            nodeName=Is On Cooldown
            activeAbilityTagsToSkipTestingValue {key=None, defaultValue {gameplayTags []}}
            gameplayAbilityTagContainer {key=None, defaultValue {gameplayTags [{tagName=NPC.Ability.NonAttack.Caretaker.EscapeIntoStorm}]}}
            targetValue {baseClass=&Actor, defaultValue=null, key=GoalActor}
           }
          }]
         properties {children [{childComposite=null, childTask=&self:BTTask_Wait_19, decorators []}], services []}
         services []
        }
       ]
       decorators [
        Has Gameplay Ability {
         class=FortBTDecorator_GameplayAbility_HasGameplayAbility
         properties {gameplayAbilityTagContainer {defaultValue {gameplayTags [{tagName=NPC.Ability.NonAttack.Wander.Escape}]}}}
        }
        Escaping Into Storm {
         class=BTDecorator_GameplayTagQuery
         properties {
          flowAbortMode=None
          nodeName=Escaping Into Storm
          gameplayTagQuery {
           autoDescription=" ANY( NPC.Ability.NonAttack.Caretaker.EscapeIntoStorm )"
           queryTokenStream [0,1,1,1,0]
           tagDictionary [{tagName=NPC.Ability.NonAttack.Caretaker.EscapeIntoStorm}]
          }
         }
        }
       ]
       properties {
        children [
         {
          childComposite=null
          childTask=&self:FortBTTask_ExecuteGameplayAbility_7
          decorators [
           &self:FortBTDecorator_GameplayAbility_CanActivate_9
           &self:BTDecorator_Blackboard_8
          ]
         }
         {
          childComposite=null
          childTask=&self:FortBTTask_ExecuteGameplayAbility_14
          decorators [
           &self:FortBTDecorator_GameplayAbility_CanActivate_12
           &self:BTDecorator_Blackboard_12
          ]
         }
         {
          childComposite=&self:BTComposite_Selector_11
          childTask=null
          decorators [&self:FortBTDecorator_GameplayAbility_IsOnCooldown_2]
         }
        ]
        services []
       }
       services []
      }
      Selector {
       class=BTComposite_Selector
       children [
        Selector {
         class=BTComposite_Selector
         children [Sequence {
           class=BTComposite_Sequence
           children [
            Rotate to face BB entry {class=BTTask_RotateToFaceBBEntry, decorators [], properties {blackboardKey {selectedKeyName=GoalActor}, precision {defaultValue=10}}}
            Wait {class=BTTask_Wait, properties {waitTime {defaultValue=0.5}}}
           ]
           decorators [
            Within Secondary Melee Approach Range? {
             class=FortBTDecorator_GameplayAbility_CompareDistance
             properties {
              bInverseCondition=false
              flowAbortMode=Self
              nodeName=Within Secondary Melee Approach Range?
              activeAbilityTagsToSkipTestingValue {key=None, defaultValue {gameplayTags []}}
              distanceComparisons [{
                addedOffset=0
                bUseAddedOffset=false
                bUseOverriddenValue=false
                comparisonType=ThreeDimensions
                operator=Less
                overriddenValue=0
                distanceDataTags {gameplayTags [{tagName=NPC.Ability.MaxApproachRange}]}
               }]
              gameplayAbilityTagContainer {key=None, defaultValue {gameplayTags [{tagName=NPC.Ability.Attack.Ranged.Primary}]}}
              targetValue {baseClass=&Actor, defaultValue=null, key=GoalActor}
             }
            }
            Could Hit if we Rotated {
             class=FortBTDecorator_GameplayAbility_CanHitTarget
             properties {
              bInverseCondition=false
              flowAbortMode=None
              nodeName=Could Hit if we Rotated
              activeAbilityTagsToSkipTestingValue {key=None, defaultValue {gameplayTags []}}
              gameplayAbilityTagContainer {key=None, defaultValue {gameplayTags [{tagName=NPC.Ability.Attack.Ranged.Primary}]}}
              targetValue {baseClass=&Actor, defaultValue=null, key=GoalActor}
              useIdealYawRotationToTargetValue {defaultValue=true, key=None}
             }
            }
            NOT Rotated To Attack Target {
             class=FortBTDecorator_GameplayAbility_IsRotatedToAttackTarget
             properties {
              bInverseCondition=true
              flowAbortMode=None
              nodeName=NOT Rotated To Attack Target
              activeAbilityTagsToSkipTestingValue {key=None, defaultValue {gameplayTags []}}
              gameplayAbilityTagContainer {key=None, defaultValue {gameplayTags [{tagName=NPC.Ability.Attack.Ranged.Primary}]}}
              targetValue {baseClass=&Actor, defaultValue=null, key=GoalActor}
             }
            }
            Is not frustrated {
             class=BTDecorator_GameplayTagQuery
             properties {
              flowAbortMode=Self
              nodeName=Is not frustrated
              gameplayTagQuery {
               autoDescription=" NONE( NPC.Ability.Attack.Caretaker.CombatFrustration.Active )"
               queryTokenStream [0,1,3,1,0]
               tagDictionary [{tagName=NPC.Ability.Attack.Caretaker.CombatFrustration.Active}]
              }
             }
            }
           ]
           properties {
            children [
             {childTask=&self:BTTask_RotateToFaceBBEntry_5}
             {childTask=&self:BTTask_Wait_11}
            ]
            services [&self:BTS_CaretakerCombatFrustration_C_6]
           }
           services [BTS_CaretakerCombatFrustration_C_6 {
             class=BTS_CaretakerCombatFrustration_C
             properties {
              bCallTickOnSearchStart=false
              bRestartTimerOnEachActivation=false
              bShowEventDetails=false
              bShowPropertyDetails=true
              customDescription=""
              defaultFrustrationDuration=5
              frustrationApplied=false
              frustrationTime=0
              interval=0.5
              nodeName=""
              randomDeviation=0.1
              maxFrustrationDuration {
               registryType=None
               value=1
               curve {curveTable=&NPCMMRScaling_Caretaker, rowName=Caretaker.Attack.General.CombatFrustrationDuration}
              }
             }
            }]
          }]
         decorators [Loop {class=BTDecorator_Loop}]
         properties {
          children [{
            childComposite=&self:BTComposite_Sequence_18
            childTask=null
            decorators [
             &self:FortBTDecorator_GameplayAbility_CompareDistance_8
             &self:FortBTDecorator_GameplayAbility_CanHitTarget_5
             &self:FortBTDecorator_GameplayAbility_IsRotatedToAttackTarget_4
             &self:BTDecorator_GameplayTagQuery_19
            ]
           }]
          services []
         }
         services []
        }
        Selector {
         class=BTComposite_Selector
         children [
          Sequence {
           class=BTComposite_Sequence
           children [
            StopMovement {
             class=BTTask_StopMovement_C
             children []
             decorators []
             properties {bIgnoreRestartSelf=false, bShowPropertyDetails=true, customDescription="", nodeName=StopMovement, tickInterval {interval=-1}}
             services []
            }
            Wait {class=BTTask_Wait, properties {waitTime {defaultValue=2.5}}}
            GameplayAbility {
             class=FortBTTask_ExecuteGameplayAbility
             decorators []
             properties {nodeName=GameplayAbility, gameplayAbilityTagContainer {defaultValue {gameplayTags [{tagName=NPC.Ability.Attack.Ranged.Primary}]}}}
            }
           ]
           decorators [IsMoving {
             class=FortBTDecorator_IsMoving
             properties {
              bInverseCondition=false
              bUseMinDist=true
              flowAbortMode=LowerPriority
              nodeName=IsMoving
              activeAbilityTagsToSkipTestingValue {key=None, defaultValue {gameplayTags []}}
              gameplayAbilityTagContainer {key=None, defaultValue {gameplayTags [{tagName=NPC.Ability.Attack.Ranged.Primary}]}}
              minDistanceComparison {
               addedOffset=0
               bUseAddedOffset=false
               bUseOverriddenValue=false
               comparisonType=ThreeDimensions
               operator=Less
               overriddenValue=0
               distanceDataTags {gameplayTags [{tagName=NPC.Ability.MinStopRange}]}
              }
              minTimeValue {defaultValue=2, key=None}
              targetValue {baseClass=&Actor, defaultValue=null, key=GoalActor}
              updateIntervalValue {defaultValue=0.1, key=None}
              useMinDistToTargetTime {defaultValue=1, key=None}
             }
            }]
           properties {children [
             {childTask=&self:BTTask_StopMovement_C_13}
             {childTask=&self:BTTask_Wait_6}
             {childTask=&self:FortBTTask_ExecuteGameplayAbility_4}
            ]}
          }
          Sequence {
           class=BTComposite_Sequence
           children [GameplayAbility {
             class=FortBTTask_ExecuteGameplayAbility
             decorators []
             properties {nodeName=GameplayAbility, gameplayAbilityTagContainer {defaultValue {gameplayTags [{tagName=NPC.Ability.Attack.Ranged.Primary}]}}}
            }]
           decorators [IsNotMoving {
             class=FortBTDecorator_IsMoving
             properties {
              bInverseCondition=true
              bUseMinDist=false
              flowAbortMode=LowerPriority
              nodeName=IsNotMoving
              activeAbilityTagsToSkipTestingValue {key=None, defaultValue {gameplayTags []}}
              gameplayAbilityTagContainer {key=None, defaultValue {gameplayTags []}}
              minDistanceComparison {
               addedOffset=0
               bUseAddedOffset=false
               bUseOverriddenValue=false
               comparisonType=TwoDimensions
               operator=Equal
               overriddenValue=0
               distanceDataTags {gameplayTags []}
              }
              minTimeValue {defaultValue=0, key=None}
              targetValue {baseClass=&Actor, defaultValue=null, key=GoalActor}
              updateIntervalValue {defaultValue=0.1, key=None}
              useMinDistToTargetTime {defaultValue=0, key=None}
             }
            }]
           properties {children [{childTask=&self:FortBTTask_ExecuteGameplayAbility_13}]}
          }
         ]
         decorators [
          "Within Ranged Primary (Swipe) Actual Range?" {
           class=FortBTDecorator_GameplayAbility_CompareDistance
           properties {
            bInverseCondition=false
            flowAbortMode=Self
            nodeName="Within Ranged Primary (Swipe) Actual Range?"
            activeAbilityTagsToSkipTestingValue {key=None, defaultValue {gameplayTags []}}
            distanceComparisons [{
              addedOffset=-50
              bUseAddedOffset=true
              bUseOverriddenValue=false
              comparisonType=ThreeDimensions
              operator=Less
              overriddenValue=0
              distanceDataTags {gameplayTags [{tagName=NPC.Ability.MaxActualRange}]}
             }]
            gameplayAbilityTagContainer {key=None, defaultValue {gameplayTags [{tagName=NPC.Ability.Attack.Ranged.Primary}]}}
            targetValue {baseClass=&Actor, defaultValue=null, key=GoalActor}
           }
          }
          Can Activate Gameplay Ability {
           class=FortBTDecorator_GameplayAbility_CanActivate
           properties {
            flowAbortMode=Self
            gameplayAbilityTagContainer {defaultValue {gameplayTags [{tagName=NPC.Ability.Attack.Ranged.Primary}]}}
            targetValue {key=GoalActor}
           }
          }
         ]
         properties {
          children [
           {
            childComposite=&self:BTComposite_Sequence_22
            childTask=null
            decorators [&self:FortBTDecorator_IsMoving_16]
           }
           {
            childComposite=&self:BTComposite_Sequence_11
            childTask=null
            decorators [&self:FortBTDecorator_IsMoving_4]
           }
          ]
          services []
         }
         services []
        }
        Sequence {
         class=BTComposite_Sequence
         children [
          CaretakerMoveTo {
           class=FortBTTask_CaretakerMoveTo
           children []
           decorators []
           properties {
            bEnableSlowdownAtGoal=false
            bFinishMoveOnOverlap=true
            bIgnoreRestartSelf=false
            bMoveDirectlyTowards=false
            bStopAtGoal=true
            nodeName=CaretakerMoveTo
            pathObstacleAction=Ignore
            acceptableRadius {defaultValue=500, key=None}
            bAllowPartialPath {defaultValue=true, key=None}
            bAllowStrafe {defaultValue=false, key=None}
            bProjectGoalLocation {defaultValue=true, key=None}
            bReachTestIncludesAgentRadius {defaultValue=true, key=None}
            bReachTestIncludesGoalRadius {defaultValue=true, key=None}
            bRequireNavigableEndLocation {defaultValue=true, key=None}
            bStartFromPreviousPath {defaultValue=false, key=None}
            bTrackMovingGoal {defaultValue=true, key=None}
            blackboardKey {bNoneIsAllowedValue=false, selectedKeyName=GoalActor}
            filterClass {
             baseClass=&NavigationQueryFilter
             defaultValue=&NavFilter_Caretaker_NoBlockers_C
             key=None
            }
            focalPointWhileMoving {bNoneIsAllowedValue=true, selectedKeyName=None}
            observedBlackboardValueTolerance {defaultValue=4.75, key=None}
           }
           services []
          }
          Wait {class=BTTask_Wait, properties {waitTime {defaultValue=0.2}}}
         ]
         decorators [
          Blackboard Based Condition {
           class=BTDecorator_Blackboard
           properties {
            arithmeticOperation=Equal
            flowAbortMode=None
            intValue=9
            nodeName=Blackboard Based Condition
            stringValue=EBotMovementState::Success
            blackboardKey {selectedKeyName=AIEvaluator_PathExists}
           }
          }
          Blackboard Based Condition {
           class=BTDecorator_Blackboard
           properties {
            arithmeticOperation=Equal
            flowAbortMode=None
            intValue=10
            nodeName=Blackboard Based Condition
            stringValue=EBotMovementState::Success_Partial
            blackboardKey {selectedKeyName=AIEvaluator_PathExists}
           }
          }
          Compare Gameplay Ability Distances {
           class=FortBTDecorator_GameplayAbility_CompareDistance
           properties {
            bInverseCondition=false
            flowAbortMode=LowerPriority
            nodeName=Compare Gameplay Ability Distances
            activeAbilityTagsToSkipTestingValue {key=None, defaultValue {gameplayTags []}}
            distanceComparisons [{
              addedOffset=0
              bUseAddedOffset=false
              bUseOverriddenValue=false
              comparisonType=ThreeDimensions
              operator=Greater
              overriddenValue=0
              distanceDataTags {gameplayTags [{tagName=NPC.Ability.MinPreferredRange}]}
             }]
            gameplayAbilityTagContainer {key=None, defaultValue {gameplayTags [{tagName=NPC.Ability.Attack.Ranged.Primary}]}}
            targetValue {baseClass=&Actor, defaultValue=null, key=GoalActor}
           }
          }
         ]
         properties {children [
           {childTask=&self:FortBTTask_CaretakerMoveTo_4}
           {childTask=&self:BTTask_Wait_10}
          ]}
        }
        Wait {
         class=BTTask_Wait
         decorators [
          Blackboard Based Condition {
           class=BTDecorator_Blackboard
           properties {
            arithmeticOperation=NotEqual
            intValue=9
            nodeName=Blackboard Based Condition
            stringValue=EBotMovementState::Success
            blackboardKey {selectedKeyName=AIEvaluator_PathExists}
           }
          }
          Blackboard Based Condition {
           class=BTDecorator_Blackboard
           properties {
            arithmeticOperation=NotEqual
            intValue=10
            nodeName=Blackboard Based Condition
            stringValue=EBotMovementState::Success_Partial
            blackboardKey {selectedKeyName=AIEvaluator_PathExists}
           }
          }
          Rotated To Attack Target {
           class=FortBTDecorator_GameplayAbility_IsRotatedToAttackTarget
           properties {
            bInverseCondition=false
            flowAbortMode=None
            nodeName=Rotated To Attack Target
            activeAbilityTagsToSkipTestingValue {key=None, defaultValue {gameplayTags []}}
            gameplayAbilityTagContainer {key=None, defaultValue {gameplayTags [{tagName=NPC.Ability.Attack.Ranged.Primary}]}}
            targetValue {baseClass=&Actor, defaultValue=null, key=GoalActor}
           }
          }
          "Within Ranged Primary (Swipe) Approach Range?" {
           class=FortBTDecorator_GameplayAbility_CompareDistance
           properties {
            bInverseCondition=false
            flowAbortMode=Self
            nodeName="Within Ranged Primary (Swipe) Approach Range?"
            activeAbilityTagsToSkipTestingValue {key=None, defaultValue {gameplayTags []}}
            distanceComparisons [{
              addedOffset=0
              bUseAddedOffset=false
              bUseOverriddenValue=false
              comparisonType=ThreeDimensions
              operator=Less
              overriddenValue=0
              distanceDataTags {gameplayTags [{tagName=NPC.Ability.MaxActualRange}]}
             }]
            gameplayAbilityTagContainer {key=None, defaultValue {gameplayTags [{tagName=NPC.Ability.Attack.Ranged.Primary}]}}
            targetValue {baseClass=&Actor, defaultValue=null, key=GoalActor}
           }
          }
         ]
         properties {waitTime {defaultValue=0.2}}
        }
       ]
       decorators [
        Has Gameplay Ability - Ranged Swipe {
         class=FortBTDecorator_GameplayAbility_HasGameplayAbility
         properties {
          nodeName=Has Gameplay Ability - Ranged Swipe
          gameplayAbilityTagContainer {defaultValue {gameplayTags [{tagName=NPC.Ability.Attack.Ranged.Primary}]}}
         }
        }
        Can Activate Gameplay Ability {
         class=FortBTDecorator_GameplayAbility_CanActivate
         properties {
          flowAbortMode=None
          gameplayAbilityTagContainer {defaultValue {gameplayTags [{tagName=NPC.Ability.Attack.Ranged.Primary}]}}
          targetValue {key=GoalActor}
         }
        }
        "Within Ranged Primary (Swipe) Approach Range?" {
         class=FortBTDecorator_GameplayAbility_CompareDistance
         properties {
          bInverseCondition=false
          flowAbortMode=LowerPriority
          nodeName="Within Ranged Primary (Swipe) Approach Range?"
          activeAbilityTagsToSkipTestingValue {key=None, defaultValue {gameplayTags []}}
          distanceComparisons [{
            addedOffset=0
            bUseAddedOffset=false
            bUseOverriddenValue=false
            comparisonType=ThreeDimensions
            operator=Less
            overriddenValue=0
            distanceDataTags {gameplayTags [{tagName=NPC.Ability.MaxApproachRange}]}
           }]
          gameplayAbilityTagContainer {key=None, defaultValue {gameplayTags [{tagName=NPC.Ability.Attack.Ranged.Primary}]}}
          targetValue {baseClass=&Actor, defaultValue=null, key=GoalActor}
         }
        }
        Not Escaping Into Storm {
         class=BTDecorator_GameplayTagQuery
         properties {
          flowAbortMode=None
          nodeName=Not Escaping Into Storm
          gameplayTagQuery {
           autoDescription=" NONE( NPC.Ability.NonAttack.Caretaker.EscapeIntoStorm )"
           queryTokenStream [0,1,3,1,0]
           tagDictionary [{tagName=NPC.Ability.NonAttack.Caretaker.EscapeIntoStorm}]
          }
         }
        }
       ]
       properties {
        children [
         {
          childComposite=&self:BTComposite_Selector_14
          childTask=null
          decorators [&self:BTDecorator_Loop_0]
         }
         {
          childComposite=&self:BTComposite_Selector_16
          childTask=null
          decorators [
           &self:FortBTDecorator_GameplayAbility_CompareDistance_25
           &self:FortBTDecorator_GameplayAbility_CanActivate_20
          ]
         }
         {
          childComposite=&self:BTComposite_Sequence_7
          childTask=null
          decoratorOps [{number=2, operation=And}, {number=2, operation=Or}, {number=0}, {number=1}, {number=2}]
          decorators [
           &self:BTDecorator_Blackboard_34
           &self:BTDecorator_Blackboard_35
           &self:FortBTDecorator_GameplayAbility_CompareDistance_12
          ]
         }
         {
          childComposite=null
          childTask=&self:BTTask_Wait_12
          decoratorOps [{number=3, operation=And}, {number=2, operation=And}, {number=0}, {number=1}, {number=2}, {number=3}]
          decorators [
           &self:BTDecorator_Blackboard_36
           &self:BTDecorator_Blackboard_37
           &self:FortBTDecorator_GameplayAbility_IsRotatedToAttackTarget_6
           &self:FortBTDecorator_GameplayAbility_CompareDistance_14
          ]
         }
        ]
        services []
       }
       services []
      }
      Selector {
       class=BTComposite_Selector
       children [
        Selector {
         class=BTComposite_Selector
         children [Sequence {
           class=BTComposite_Sequence
           children [
            Rotate to face BB entry {class=BTTask_RotateToFaceBBEntry, decorators [], properties {blackboardKey {selectedKeyName=GoalActor}, precision {defaultValue=10}}}
            Wait {class=BTTask_Wait, properties {waitTime {defaultValue=0.5}}}
           ]
           decorators [
            Within Secondary Melee Approach Range? {
             class=FortBTDecorator_GameplayAbility_CompareDistance
             properties {
              bInverseCondition=false
              flowAbortMode=Self
              nodeName=Within Secondary Melee Approach Range?
              activeAbilityTagsToSkipTestingValue {key=None, defaultValue {gameplayTags []}}
              distanceComparisons [{
                addedOffset=0
                bUseAddedOffset=false
                bUseOverriddenValue=false
                comparisonType=ThreeDimensions
                operator=Less
                overriddenValue=0
                distanceDataTags {gameplayTags [{tagName=NPC.Ability.MaxApproachRange}]}
               }]
              gameplayAbilityTagContainer {key=None, defaultValue {gameplayTags [{tagName=NPC.Ability.Attack.Melee.Secondary}]}}
              targetValue {baseClass=&Actor, defaultValue=null, key=GoalActor}
             }
            }
            Could Hit if we Rotated {
             class=FortBTDecorator_GameplayAbility_CanHitTarget
             properties {
              bInverseCondition=false
              flowAbortMode=None
              nodeName=Could Hit if we Rotated
              activeAbilityTagsToSkipTestingValue {key=None, defaultValue {gameplayTags []}}
              gameplayAbilityTagContainer {key=None, defaultValue {gameplayTags [{tagName=NPC.Ability.Attack.Melee.Secondary}]}}
              targetValue {baseClass=&Actor, defaultValue=null, key=GoalActor}
              useIdealYawRotationToTargetValue {defaultValue=true, key=None}
             }
            }
            NOT Rotated To Attack Target {
             class=FortBTDecorator_GameplayAbility_IsRotatedToAttackTarget
             properties {
              bInverseCondition=true
              flowAbortMode=None
              nodeName=NOT Rotated To Attack Target
              activeAbilityTagsToSkipTestingValue {key=None, defaultValue {gameplayTags []}}
              gameplayAbilityTagContainer {key=None, defaultValue {gameplayTags [{tagName=NPC.Ability.Attack.Melee.Secondary}]}}
              targetValue {baseClass=&Actor, defaultValue=null, key=GoalActor}
             }
            }
            Is not frustrated {
             class=BTDecorator_GameplayTagQuery
             properties {
              flowAbortMode=Self
              nodeName=Is not frustrated
              gameplayTagQuery {
               autoDescription=" NONE( NPC.Ability.Attack.Caretaker.CombatFrustration.Active )"
               queryTokenStream [0,1,3,1,0]
               tagDictionary [{tagName=NPC.Ability.Attack.Caretaker.CombatFrustration.Active}]
              }
             }
            }
           ]
           properties {
            children [
             {childTask=&self:BTTask_RotateToFaceBBEntry_1}
             {childTask=&self:BTTask_Wait_4}
            ]
            services [&self:BTS_CaretakerCombatFrustration_C_1]
           }
           services [BTS_CaretakerCombatFrustration_C_1 {
             class=BTS_CaretakerCombatFrustration_C
             properties {
              bCallTickOnSearchStart=false
              bRestartTimerOnEachActivation=false
              bShowEventDetails=false
              bShowPropertyDetails=true
              customDescription=""
              defaultFrustrationDuration=5
              frustrationApplied=false
              frustrationTime=0
              interval=0.5
              nodeName=""
              randomDeviation=0.1
              maxFrustrationDuration {
               registryType=None
               value=1
               curve {curveTable=&NPCMMRScaling_Caretaker, rowName=Caretaker.Attack.General.CombatFrustrationDuration}
              }
             }
            }]
          }]
         decorators [Loop {class=BTDecorator_Loop}]
         properties {
          children [{
            childComposite=&self:BTComposite_Sequence_2
            childTask=null
            decorators [
             &self:FortBTDecorator_GameplayAbility_CompareDistance_18
             &self:FortBTDecorator_GameplayAbility_CanHitTarget_4
             &self:FortBTDecorator_GameplayAbility_IsRotatedToAttackTarget_8
             &self:BTDecorator_GameplayTagQuery_8
            ]
           }]
          services []
         }
         services []
        }
        Selector {
         class=BTComposite_Selector
         children [
          Sequence {
           class=BTComposite_Sequence
           children [
            StopMovement {
             class=BTTask_StopMovement_C
             children []
             decorators []
             properties {bIgnoreRestartSelf=false, bShowPropertyDetails=true, customDescription="", nodeName=StopMovement, tickInterval {interval=-1}}
             services []
            }
            Wait {class=BTTask_Wait, properties {waitTime {defaultValue=2.5}}}
            GameplayAbility {
             class=FortBTTask_ExecuteGameplayAbility
             decorators []
             properties {nodeName=GameplayAbility, gameplayAbilityTagContainer {defaultValue {gameplayTags [{tagName=NPC.Ability.Attack.Melee.Secondary}]}}}
            }
           ]
           decorators [IsMoving {
             class=FortBTDecorator_IsMoving
             properties {
              bInverseCondition=false
              bUseMinDist=true
              flowAbortMode=LowerPriority
              nodeName=IsMoving
              activeAbilityTagsToSkipTestingValue {key=None, defaultValue {gameplayTags []}}
              gameplayAbilityTagContainer {key=None, defaultValue {gameplayTags [{tagName=NPC.Ability.Attack.Melee.Secondary}]}}
              minDistanceComparison {
               addedOffset=0
               bUseAddedOffset=false
               bUseOverriddenValue=false
               comparisonType=ThreeDimensions
               operator=Less
               overriddenValue=0
               distanceDataTags {gameplayTags [{tagName=NPC.Ability.MinStopRange}]}
              }
              minTimeValue {defaultValue=2, key=None}
              targetValue {baseClass=&Actor, defaultValue=null, key=GoalActor}
              updateIntervalValue {defaultValue=0.1, key=None}
              useMinDistToTargetTime {defaultValue=1, key=None}
             }
            }]
           properties {children [
             {childTask=&self:BTTask_StopMovement_C_0}
             {childTask=&self:BTTask_Wait_13}
             {childTask=&self:FortBTTask_ExecuteGameplayAbility_9}
            ]}
          }
          Sequence {
           class=BTComposite_Sequence
           children [GameplayAbility {
             class=FortBTTask_ExecuteGameplayAbility
             decorators []
             properties {nodeName=GameplayAbility, gameplayAbilityTagContainer {defaultValue {gameplayTags [{tagName=NPC.Ability.Attack.Melee.Secondary}]}}}
            }]
           decorators [IsNotMoving {
             class=FortBTDecorator_IsMoving
             properties {
              bInverseCondition=true
              bUseMinDist=false
              flowAbortMode=LowerPriority
              nodeName=IsNotMoving
              activeAbilityTagsToSkipTestingValue {key=None, defaultValue {gameplayTags []}}
              gameplayAbilityTagContainer {key=None, defaultValue {gameplayTags []}}
              minDistanceComparison {
               addedOffset=0
               bUseAddedOffset=false
               bUseOverriddenValue=false
               comparisonType=TwoDimensions
               operator=Equal
               overriddenValue=0
               distanceDataTags {gameplayTags []}
              }
              minTimeValue {defaultValue=0, key=None}
              targetValue {baseClass=&Actor, defaultValue=null, key=GoalActor}
              updateIntervalValue {defaultValue=0.1, key=None}
              useMinDistToTargetTime {defaultValue=0, key=None}
             }
            }]
           properties {children [{childTask=&self:FortBTTask_ExecuteGameplayAbility_10}]}
          }
         ]
         decorators [
          Within Secondary Melee Actual Range? {
           class=FortBTDecorator_GameplayAbility_CompareDistance
           properties {
            bInverseCondition=false
            flowAbortMode=Self
            nodeName=Within Secondary Melee Actual Range?
            activeAbilityTagsToSkipTestingValue {key=None, defaultValue {gameplayTags []}}
            distanceComparisons [{
              addedOffset=-50
              bUseAddedOffset=true
              bUseOverriddenValue=false
              comparisonType=ThreeDimensions
              operator=Less
              overriddenValue=0
              distanceDataTags {gameplayTags [{tagName=NPC.Ability.MaxActualRange}]}
             }]
            gameplayAbilityTagContainer {key=None, defaultValue {gameplayTags [{tagName=NPC.Ability.Attack.Melee.Secondary}]}}
            targetValue {baseClass=&Actor, defaultValue=null, key=GoalActor}
           }
          }
          Can Activate Gameplay Ability {
           class=FortBTDecorator_GameplayAbility_CanActivate
           properties {
            flowAbortMode=Self
            gameplayAbilityTagContainer {defaultValue {gameplayTags [{tagName=NPC.Ability.Attack.Melee.Secondary}]}}
            targetValue {key=GoalActor}
           }
          }
         ]
         properties {
          children [
           {
            childComposite=&self:BTComposite_Sequence_12
            childTask=null
            decorators [&self:FortBTDecorator_IsMoving_3]
           }
           {
            childComposite=&self:BTComposite_Sequence_3
            childTask=null
            decorators [&self:FortBTDecorator_IsMoving_0]
           }
          ]
          services []
         }
         services []
        }
        Sequence {
         class=BTComposite_Sequence
         children [
          CaretakerMoveTo {
           class=FortBTTask_CaretakerMoveTo
           children []
           decorators []
           properties {
            bEnableSlowdownAtGoal=false
            bFinishMoveOnOverlap=true
            bIgnoreRestartSelf=false
            bMoveDirectlyTowards=false
            bStopAtGoal=true
            nodeName=CaretakerMoveTo
            pathObstacleAction=Ignore
            acceptableRadius {defaultValue=500, key=None}
            bAllowPartialPath {defaultValue=true, key=None}
            bAllowStrafe {defaultValue=false, key=None}
            bProjectGoalLocation {defaultValue=true, key=None}
            bReachTestIncludesAgentRadius {defaultValue=true, key=None}
            bReachTestIncludesGoalRadius {defaultValue=true, key=None}
            bRequireNavigableEndLocation {defaultValue=true, key=None}
            bStartFromPreviousPath {defaultValue=false, key=None}
            bTrackMovingGoal {defaultValue=true, key=None}
            blackboardKey {bNoneIsAllowedValue=false, selectedKeyName=GoalActor}
            filterClass {
             baseClass=&NavigationQueryFilter
             defaultValue=&NavFilter_Caretaker_NoBlockers_C
             key=None
            }
            focalPointWhileMoving {bNoneIsAllowedValue=true, selectedKeyName=None}
            observedBlackboardValueTolerance {defaultValue=4.75, key=None}
           }
           services []
          }
          Wait {class=BTTask_Wait, properties {waitTime {defaultValue=0.2}}}
         ]
         decorators [
          Blackboard Based Condition {
           class=BTDecorator_Blackboard
           properties {
            arithmeticOperation=Equal
            flowAbortMode=None
            intValue=9
            nodeName=Blackboard Based Condition
            stringValue=EBotMovementState::Success
            blackboardKey {selectedKeyName=AIEvaluator_PathExists}
           }
          }
          Blackboard Based Condition {
           class=BTDecorator_Blackboard
           properties {
            arithmeticOperation=Equal
            flowAbortMode=None
            intValue=10
            nodeName=Blackboard Based Condition
            stringValue=EBotMovementState::Success_Partial
            blackboardKey {selectedKeyName=AIEvaluator_PathExists}
           }
          }
          Compare Gameplay Ability Distances {
           class=FortBTDecorator_GameplayAbility_CompareDistance
           properties {
            bInverseCondition=false
            flowAbortMode=LowerPriority
            nodeName=Compare Gameplay Ability Distances
            activeAbilityTagsToSkipTestingValue {key=None, defaultValue {gameplayTags []}}
            distanceComparisons [{
              addedOffset=0
              bUseAddedOffset=false
              bUseOverriddenValue=false
              comparisonType=ThreeDimensions
              operator=Greater
              overriddenValue=0
              distanceDataTags {gameplayTags [{tagName=NPC.Ability.MinPreferredRange}]}
             }]
            gameplayAbilityTagContainer {key=None, defaultValue {gameplayTags [{tagName=NPC.Ability.Attack.Melee.Secondary}]}}
            targetValue {baseClass=&Actor, defaultValue=null, key=GoalActor}
           }
          }
         ]
         properties {children [
           {childTask=&self:FortBTTask_CaretakerMoveTo_1}
           {childTask=&self:BTTask_Wait_0}
          ]}
        }
        Wait {
         class=BTTask_Wait
         decorators [
          Blackboard Based Condition {
           class=BTDecorator_Blackboard
           properties {
            arithmeticOperation=NotEqual
            intValue=9
            nodeName=Blackboard Based Condition
            stringValue=EBotMovementState::Success
            blackboardKey {selectedKeyName=AIEvaluator_PathExists}
           }
          }
          Blackboard Based Condition {
           class=BTDecorator_Blackboard
           properties {
            arithmeticOperation=NotEqual
            intValue=10
            nodeName=Blackboard Based Condition
            stringValue=EBotMovementState::Success_Partial
            blackboardKey {selectedKeyName=AIEvaluator_PathExists}
           }
          }
          Rotated To Attack Target {
           class=FortBTDecorator_GameplayAbility_IsRotatedToAttackTarget
           properties {
            bInverseCondition=false
            flowAbortMode=None
            nodeName=Rotated To Attack Target
            activeAbilityTagsToSkipTestingValue {key=None, defaultValue {gameplayTags []}}
            gameplayAbilityTagContainer {key=None, defaultValue {gameplayTags [{tagName=NPC.Ability.Attack.Melee.Secondary}]}}
            targetValue {baseClass=&Actor, defaultValue=null, key=GoalActor}
           }
          }
          Within Secondary Melee Approach Range? {
           class=FortBTDecorator_GameplayAbility_CompareDistance
           properties {
            bInverseCondition=false
            flowAbortMode=Self
            nodeName=Within Secondary Melee Approach Range?
            activeAbilityTagsToSkipTestingValue {key=None, defaultValue {gameplayTags []}}
            distanceComparisons [{
              addedOffset=0
              bUseAddedOffset=false
              bUseOverriddenValue=false
              comparisonType=ThreeDimensions
              operator=Less
              overriddenValue=0
              distanceDataTags {gameplayTags [{tagName=NPC.Ability.MaxActualRange}]}
             }]
            gameplayAbilityTagContainer {key=None, defaultValue {gameplayTags [{tagName=NPC.Ability.Attack.Melee.Secondary}]}}
            targetValue {baseClass=&Actor, defaultValue=null, key=GoalActor}
           }
          }
         ]
         properties {waitTime {defaultValue=0.2}}
        }
       ]
       decorators [
        "Has Gameplay Ability - Melee Secondary (Snap)" {
         class=FortBTDecorator_GameplayAbility_HasGameplayAbility
         properties {
          nodeName="Has Gameplay Ability - Melee Secondary (Snap)"
          gameplayAbilityTagContainer {defaultValue {gameplayTags [{tagName=NPC.Ability.Attack.Melee.Secondary}]}}
         }
        }
        Can Activate Gameplay Ability {
         class=FortBTDecorator_GameplayAbility_CanActivate
         properties {
          flowAbortMode=None
          gameplayAbilityTagContainer {defaultValue {gameplayTags [{tagName=NPC.Ability.Attack.Melee.Secondary}]}}
          targetValue {key=GoalActor}
         }
        }
        Within Secondary Melee Approach Range? {
         class=FortBTDecorator_GameplayAbility_CompareDistance
         properties {
          bInverseCondition=false
          flowAbortMode=LowerPriority
          nodeName=Within Secondary Melee Approach Range?
          activeAbilityTagsToSkipTestingValue {key=None, defaultValue {gameplayTags []}}
          distanceComparisons [{
            addedOffset=0
            bUseAddedOffset=false
            bUseOverriddenValue=false
            comparisonType=ThreeDimensions
            operator=Less
            overriddenValue=0
            distanceDataTags {gameplayTags [{tagName=NPC.Ability.MaxApproachRange}]}
           }]
          gameplayAbilityTagContainer {key=None, defaultValue {gameplayTags [{tagName=NPC.Ability.Attack.Melee.Secondary}]}}
          targetValue {baseClass=&Actor, defaultValue=null, key=GoalActor}
         }
        }
        Not Escaping Into Storm {
         class=BTDecorator_GameplayTagQuery
         properties {
          flowAbortMode=None
          nodeName=Not Escaping Into Storm
          gameplayTagQuery {
           autoDescription=" NONE( NPC.Ability.NonAttack.Caretaker.EscapeIntoStorm )"
           queryTokenStream [0,1,3,1,0]
           tagDictionary [{tagName=NPC.Ability.NonAttack.Caretaker.EscapeIntoStorm}]
          }
         }
        }
       ]
       properties {
        children [
         {
          childComposite=&self:BTComposite_Selector_20
          childTask=null
          decorators [&self:BTDecorator_Loop_5]
         }
         {
          childComposite=&self:BTComposite_Selector_15
          childTask=null
          decorators [
           &self:FortBTDecorator_GameplayAbility_CompareDistance_1
           &self:FortBTDecorator_GameplayAbility_CanActivate_4
          ]
         }
         {
          childComposite=&self:BTComposite_Sequence_0
          childTask=null
          decoratorOps [{number=2, operation=And}, {number=2, operation=Or}, {number=0}, {number=1}, {number=2}]
          decorators [
           &self:BTDecorator_Blackboard_28
           &self:BTDecorator_Blackboard_29
           &self:FortBTDecorator_GameplayAbility_CompareDistance_17
          ]
         }
         {
          childComposite=null
          childTask=&self:BTTask_Wait_3
          decoratorOps [{number=3, operation=And}, {number=2, operation=And}, {number=0}, {number=1}, {number=2}, {number=3}]
          decorators [
           &self:BTDecorator_Blackboard_32
           &self:BTDecorator_Blackboard_33
           &self:FortBTDecorator_GameplayAbility_IsRotatedToAttackTarget_9
           &self:FortBTDecorator_GameplayAbility_CompareDistance_19
          ]
         }
        ]
        services []
       }
       services []
      }
      Selector {
       class=BTComposite_Selector
       children [
        Selector {
         class=BTComposite_Selector
         children [Sequence {
           class=BTComposite_Sequence
           children [
            Rotate to face BB entry {
             class=BTTask_RotateToFaceBBEntry
             decorators [NOT Rotated To Attack Target {
               class=FortBTDecorator_GameplayAbility_IsRotatedToAttackTarget
               properties {
                bInverseCondition=true
                flowAbortMode=None
                nodeName=NOT Rotated To Attack Target
                activeAbilityTagsToSkipTestingValue {key=None, defaultValue {gameplayTags []}}
                gameplayAbilityTagContainer {key=None, defaultValue {gameplayTags [{tagName=NPC.Ability.Attack.Melee.Primary}]}}
                targetValue {baseClass=&Actor, defaultValue=null, key=GoalActor}
               }
              }]
             properties {blackboardKey {selectedKeyName=GoalActor}, precision {defaultValue=10}}
            }
            Wait {class=BTTask_Wait, properties {waitTime {defaultValue=0.5}}}
           ]
           decorators [
            Within Primary Melee Approach Range? {
             class=FortBTDecorator_GameplayAbility_CompareDistance
             properties {
              bInverseCondition=false
              flowAbortMode=Self
              nodeName=Within Primary Melee Approach Range?
              activeAbilityTagsToSkipTestingValue {key=None, defaultValue {gameplayTags []}}
              distanceComparisons [{
                addedOffset=0
                bUseAddedOffset=false
                bUseOverriddenValue=false
                comparisonType=ThreeDimensions
                operator=Less
                overriddenValue=0
                distanceDataTags {gameplayTags [{tagName=NPC.Ability.MaxApproachRange}]}
               }]
              gameplayAbilityTagContainer {key=None, defaultValue {gameplayTags [{tagName=NPC.Ability.Attack.Melee.Primary}]}}
              targetValue {baseClass=&Actor, defaultValue=null, key=GoalActor}
             }
            }
            Could Hit if we Rotated {
             class=FortBTDecorator_GameplayAbility_CanHitTarget
             properties {
              bInverseCondition=false
              flowAbortMode=None
              nodeName=Could Hit if we Rotated
              activeAbilityTagsToSkipTestingValue {key=None, defaultValue {gameplayTags []}}
              gameplayAbilityTagContainer {key=None, defaultValue {gameplayTags [{tagName=NPC.Ability.Attack.Melee.Primary}]}}
              targetValue {baseClass=&Actor, defaultValue=null, key=GoalActor}
              useIdealYawRotationToTargetValue {defaultValue=true, key=None}
             }
            }
            Is not frustrated {
             class=BTDecorator_GameplayTagQuery
             properties {
              flowAbortMode=Self
              nodeName=Is not frustrated
              gameplayTagQuery {
               autoDescription=" NONE( NPC.Ability.Attack.Caretaker.CombatFrustration.Active )"
               queryTokenStream [0,1,3,1,0]
               tagDictionary [{tagName=NPC.Ability.Attack.Caretaker.CombatFrustration.Active}]
              }
             }
            }
           ]
           properties {
            children [
             {
              childTask=&self:BTTask_RotateToFaceBBEntry_6
              decorators [&self:FortBTDecorator_GameplayAbility_IsRotatedToAttackTarget_3]
             }
             {childTask=&self:BTTask_Wait_14}
            ]
            services [&self:BTS_CaretakerCombatFrustration_C_4]
           }
           services [BTS_CaretakerCombatFrustration_C_4 {
             class=BTS_CaretakerCombatFrustration_C
             properties {
              bCallTickOnSearchStart=false
              bRestartTimerOnEachActivation=false
              bShowEventDetails=false
              bShowPropertyDetails=true
              customDescription=""
              defaultFrustrationDuration=5
              frustrationApplied=false
              frustrationTime=0
              interval=0.5
              nodeName=""
              randomDeviation=0.1
              maxFrustrationDuration {
               registryType=None
               value=1
               curve {curveTable=&NPCMMRScaling_Caretaker, rowName=Caretaker.Attack.General.CombatFrustrationDuration}
              }
             }
            }]
          }]
         decorators [Loop {class=BTDecorator_Loop}]
         properties {
          children [{
            childComposite=&self:BTComposite_Sequence_16
            childTask=null
            decorators [
             &self:FortBTDecorator_GameplayAbility_CompareDistance_4
             &self:FortBTDecorator_GameplayAbility_CanHitTarget_1
             &self:BTDecorator_GameplayTagQuery_15
            ]
           }]
          services []
         }
         services []
        }
        Selector {
         class=BTComposite_Selector
         children [
          Sequence {
           class=BTComposite_Sequence
           children [
            StopMovement {
             class=BTTask_StopMovement_C
             children []
             decorators []
             properties {bIgnoreRestartSelf=false, bShowPropertyDetails=true, customDescription="", nodeName=StopMovement, tickInterval {interval=-1}}
             services []
            }
            Wait {class=BTTask_Wait, properties {waitTime {defaultValue=2.5}}}
            GameplayAbility {
             class=FortBTTask_ExecuteGameplayAbility
             decorators []
             properties {nodeName=GameplayAbility, gameplayAbilityTagContainer {defaultValue {gameplayTags [{tagName=NPC.Ability.Attack.Melee.Primary}]}}}
            }
           ]
           decorators [IsMoving {
             class=FortBTDecorator_IsMoving
             properties {
              bInverseCondition=false
              bUseMinDist=true
              flowAbortMode=LowerPriority
              nodeName=IsMoving
              activeAbilityTagsToSkipTestingValue {key=None, defaultValue {gameplayTags []}}
              gameplayAbilityTagContainer {key=None, defaultValue {gameplayTags [{tagName=NPC.Ability.Attack.Melee.Primary}]}}
              minDistanceComparison {
               addedOffset=0
               bUseAddedOffset=false
               bUseOverriddenValue=false
               comparisonType=ThreeDimensions
               operator=Less
               overriddenValue=0
               distanceDataTags {gameplayTags [{tagName=NPC.Ability.MinStopRange}]}
              }
              minTimeValue {defaultValue=2, key=None}
              targetValue {baseClass=&Actor, defaultValue=null, key=GoalActor}
              updateIntervalValue {defaultValue=0.1, key=None}
              useMinDistToTargetTime {defaultValue=1, key=None}
             }
            }]
           properties {children [
             {childTask=&self:BTTask_StopMovement_C_2}
             {childTask=&self:BTTask_Wait_9}
             {childTask=&self:FortBTTask_ExecuteGameplayAbility_15}
            ]}
          }
          Sequence {
           class=BTComposite_Sequence
           children [GameplayAbility {
             class=FortBTTask_ExecuteGameplayAbility
             decorators []
             properties {nodeName=GameplayAbility, gameplayAbilityTagContainer {defaultValue {gameplayTags [{tagName=NPC.Ability.Attack.Melee.Primary}]}}}
            }]
           decorators [IsNotMoving {
             class=FortBTDecorator_IsMoving
             properties {
              bInverseCondition=true
              bUseMinDist=false
              flowAbortMode=LowerPriority
              nodeName=IsNotMoving
              activeAbilityTagsToSkipTestingValue {key=None, defaultValue {gameplayTags []}}
              gameplayAbilityTagContainer {key=None, defaultValue {gameplayTags []}}
              minDistanceComparison {
               addedOffset=0
               bUseAddedOffset=false
               bUseOverriddenValue=false
               comparisonType=TwoDimensions
               operator=Equal
               overriddenValue=0
               distanceDataTags {gameplayTags []}
              }
              minTimeValue {defaultValue=0, key=None}
              targetValue {baseClass=&Actor, defaultValue=null, key=GoalActor}
              updateIntervalValue {defaultValue=0.1, key=None}
              useMinDistToTargetTime {defaultValue=0, key=None}
             }
            }]
           properties {children [{childTask=&self:FortBTTask_ExecuteGameplayAbility_12}]}
          }
         ]
         decorators [
          Within Primary Melee Actual Range {
           class=FortBTDecorator_GameplayAbility_CompareDistance
           properties {
            bInverseCondition=false
            flowAbortMode=Self
            nodeName=Within Primary Melee Actual Range
            activeAbilityTagsToSkipTestingValue {key=None, defaultValue {gameplayTags []}}
            distanceComparisons [{
              addedOffset=0
              bUseAddedOffset=false
              bUseOverriddenValue=false
              comparisonType=ThreeDimensions
              operator=Less
              overriddenValue=0
              distanceDataTags {gameplayTags [{tagName=NPC.Ability.MaxActualRange}]}
             }]
            gameplayAbilityTagContainer {key=None, defaultValue {gameplayTags [{tagName=NPC.Ability.Attack.Melee.Primary}]}}
            targetValue {baseClass=&Actor, defaultValue=null, key=GoalActor}
           }
          }
          Could Hit {
           class=FortBTDecorator_GameplayAbility_CanHitTarget
           properties {
            bInverseCondition=false
            flowAbortMode=Self
            nodeName=Could Hit
            activeAbilityTagsToSkipTestingValue {key=None, defaultValue {gameplayTags []}}
            gameplayAbilityTagContainer {key=None, defaultValue {gameplayTags [{tagName=NPC.Ability.Attack.Melee.Primary}]}}
            targetValue {baseClass=&Actor, defaultValue=null, key=GoalActor}
            useIdealYawRotationToTargetValue {defaultValue=true, key=None}
           }
          }
          Can Activate Gameplay Ability {
           class=FortBTDecorator_GameplayAbility_CanActivate
           properties {
            flowAbortMode=Self
            gameplayAbilityTagContainer {defaultValue {gameplayTags [{tagName=NPC.Ability.Attack.Melee.Primary}]}}
            targetValue {key=GoalActor}
           }
          }
         ]
         properties {
          children [
           {
            childComposite=&self:BTComposite_Sequence_25
            childTask=null
            decorators [&self:FortBTDecorator_IsMoving_19]
           }
           {
            childComposite=&self:BTComposite_Sequence_9
            childTask=null
            decorators [&self:FortBTDecorator_IsMoving_1]
           }
          ]
          services []
         }
         services []
        }
        Sequence {
         class=BTComposite_Sequence
         children [
          CaretakerMoveTo {
           class=FortBTTask_CaretakerMoveTo
           children []
           decorators []
           properties {
            bEnableSlowdownAtGoal=false
            bFinishMoveOnOverlap=true
            bIgnoreRestartSelf=false
            bMoveDirectlyTowards=false
            bStopAtGoal=true
            nodeName=CaretakerMoveTo
            pathObstacleAction=Ignore
            acceptableRadius {defaultValue=500, key=None}
            bAllowPartialPath {defaultValue=true, key=None}
            bAllowStrafe {defaultValue=false, key=None}
            bProjectGoalLocation {defaultValue=true, key=None}
            bReachTestIncludesAgentRadius {defaultValue=true, key=None}
            bReachTestIncludesGoalRadius {defaultValue=true, key=None}
            bRequireNavigableEndLocation {defaultValue=true, key=None}
            bStartFromPreviousPath {defaultValue=false, key=None}
            bTrackMovingGoal {defaultValue=true, key=None}
            blackboardKey {bNoneIsAllowedValue=false, selectedKeyName=GoalActor}
            filterClass {
             baseClass=&NavigationQueryFilter
             defaultValue=&NavFilter_Caretaker_NoBlockers_C
             key=None
            }
            focalPointWhileMoving {bNoneIsAllowedValue=true, selectedKeyName=None}
            observedBlackboardValueTolerance {defaultValue=4.75, key=None}
           }
           services []
          }
          Wait {class=BTTask_Wait, properties {waitTime {defaultValue=0.2}}}
         ]
         decorators [
          Blackboard Based Condition {
           class=BTDecorator_Blackboard
           properties {
            arithmeticOperation=Equal
            flowAbortMode=None
            intValue=9
            nodeName=Blackboard Based Condition
            stringValue=EBotMovementState::Success
            blackboardKey {selectedKeyName=AIEvaluator_PathExists}
           }
          }
          Blackboard Based Condition {
           class=BTDecorator_Blackboard
           properties {
            arithmeticOperation=Equal
            flowAbortMode=None
            intValue=10
            nodeName=Blackboard Based Condition
            stringValue=EBotMovementState::Success_Partial
            blackboardKey {selectedKeyName=AIEvaluator_PathExists}
           }
          }
          Compare Gameplay Ability Distances {
           class=FortBTDecorator_GameplayAbility_CompareDistance
           properties {
            bInverseCondition=false
            flowAbortMode=LowerPriority
            nodeName=Compare Gameplay Ability Distances
            activeAbilityTagsToSkipTestingValue {key=None, defaultValue {gameplayTags []}}
            distanceComparisons [{
              addedOffset=0
              bUseAddedOffset=false
              bUseOverriddenValue=false
              comparisonType=ThreeDimensions
              operator=Greater
              overriddenValue=0
              distanceDataTags {gameplayTags [{tagName=NPC.Ability.MinPreferredRange}]}
             }]
            gameplayAbilityTagContainer {key=None, defaultValue {gameplayTags [{tagName=NPC.Ability.Attack.Melee.Primary}]}}
            targetValue {baseClass=&Actor, defaultValue=null, key=GoalActor}
           }
          }
         ]
         properties {children [
           {childTask=&self:FortBTTask_CaretakerMoveTo_3}
           {childTask=&self:BTTask_Wait_7}
          ]}
        }
        Wait {
         class=BTTask_Wait
         decorators [
          Blackboard Based Condition {
           class=BTDecorator_Blackboard
           properties {
            arithmeticOperation=NotEqual
            intValue=9
            nodeName=Blackboard Based Condition
            stringValue=EBotMovementState::Success
            blackboardKey {selectedKeyName=AIEvaluator_PathExists}
           }
          }
          Blackboard Based Condition {
           class=BTDecorator_Blackboard
           properties {
            arithmeticOperation=NotEqual
            intValue=10
            nodeName=Blackboard Based Condition
            stringValue=EBotMovementState::Success_Partial
            blackboardKey {selectedKeyName=AIEvaluator_PathExists}
           }
          }
          Rotated To Attack Target {
           class=FortBTDecorator_GameplayAbility_IsRotatedToAttackTarget
           properties {
            bInverseCondition=false
            flowAbortMode=None
            nodeName=Rotated To Attack Target
            activeAbilityTagsToSkipTestingValue {key=None, defaultValue {gameplayTags []}}
            gameplayAbilityTagContainer {key=None, defaultValue {gameplayTags [{tagName=NPC.Ability.Attack.Melee.Primary}]}}
            targetValue {baseClass=&Actor, defaultValue=null, key=GoalActor}
           }
          }
          Within Primary Melee Approach Range? {
           class=FortBTDecorator_GameplayAbility_CompareDistance
           properties {
            bInverseCondition=false
            flowAbortMode=Self
            nodeName=Within Primary Melee Approach Range?
            activeAbilityTagsToSkipTestingValue {key=None, defaultValue {gameplayTags []}}
            distanceComparisons [{
              addedOffset=0
              bUseAddedOffset=false
              bUseOverriddenValue=false
              comparisonType=ThreeDimensions
              operator=Less
              overriddenValue=0
              distanceDataTags {gameplayTags [{tagName=NPC.Ability.MaxActualRange}]}
             }]
            gameplayAbilityTagContainer {key=None, defaultValue {gameplayTags [{tagName=NPC.Ability.Attack.Melee.Primary}]}}
            targetValue {baseClass=&Actor, defaultValue=null, key=GoalActor}
           }
          }
         ]
         properties {waitTime {defaultValue=0.2}}
        }
       ]
       decorators [
        Has Gameplay Ability - Melee Primary {
         class=FortBTDecorator_GameplayAbility_HasGameplayAbility
         properties {
          nodeName=Has Gameplay Ability - Melee Primary
          gameplayAbilityTagContainer {defaultValue {gameplayTags [{tagName=NPC.Ability.Attack.Melee.Primary}]}}
         }
        }
        Can Activate Gameplay Ability {
         class=FortBTDecorator_GameplayAbility_CanActivate
         properties {
          flowAbortMode=None
          gameplayAbilityTagContainer {defaultValue {gameplayTags [{tagName=NPC.Ability.Attack.Melee.Primary}]}}
          targetValue {key=GoalActor}
         }
        }
        Within Primary Melee Approach Range? {
         class=FortBTDecorator_GameplayAbility_CompareDistance
         properties {
          bInverseCondition=false
          flowAbortMode=LowerPriority
          nodeName=Within Primary Melee Approach Range?
          activeAbilityTagsToSkipTestingValue {key=None, defaultValue {gameplayTags []}}
          distanceComparisons [{
            addedOffset=0
            bUseAddedOffset=false
            bUseOverriddenValue=false
            comparisonType=ThreeDimensions
            operator=Less
            overriddenValue=0
            distanceDataTags {gameplayTags [{tagName=NPC.Ability.MaxApproachRange}]}
           }]
          gameplayAbilityTagContainer {key=None, defaultValue {gameplayTags [{tagName=NPC.Ability.Attack.Melee.Primary}]}}
          targetValue {baseClass=&Actor, defaultValue=null, key=GoalActor}
         }
        }
        Not Escaping Into Storm {
         class=BTDecorator_GameplayTagQuery
         properties {
          flowAbortMode=None
          nodeName=Not Escaping Into Storm
          gameplayTagQuery {
           autoDescription=" NONE( NPC.Ability.NonAttack.Caretaker.EscapeIntoStorm )"
           queryTokenStream [0,1,3,1,0]
           tagDictionary [{tagName=NPC.Ability.NonAttack.Caretaker.EscapeIntoStorm}]
          }
         }
        }
       ]
       properties {
        children [
         {
          childComposite=&self:BTComposite_Selector_6
          childTask=null
          decorators [&self:BTDecorator_Loop_2]
         }
         {
          childComposite=&self:BTComposite_Selector_18
          childTask=null
          decorators [
           &self:FortBTDecorator_GameplayAbility_CompareDistance_22
           &self:FortBTDecorator_GameplayAbility_CanHitTarget_7
           &self:FortBTDecorator_GameplayAbility_CanActivate_19
          ]
         }
         {
          childComposite=&self:BTComposite_Sequence_5
          childTask=null
          decoratorOps [{number=2, operation=And}, {number=2, operation=Or}, {number=0}, {number=1}, {number=2}]
          decorators [
           &self:BTDecorator_Blackboard_20
           &self:BTDecorator_Blackboard_21
           &self:FortBTDecorator_GameplayAbility_CompareDistance_3
          ]
         }
         {
          childComposite=null
          childTask=&self:BTTask_Wait_8
          decoratorOps [{number=3, operation=And}, {number=2, operation=And}, {number=0}, {number=1}, {number=2}, {number=3}]
          decorators [
           &self:BTDecorator_Blackboard_10
           &self:BTDecorator_Blackboard_11
           &self:FortBTDecorator_GameplayAbility_IsRotatedToAttackTarget_5
           &self:FortBTDecorator_GameplayAbility_CompareDistance_11
          ]
         }
        ]
        services []
       }
       services []
      }
     ]
     decorators []
     properties {
      nodeName=Selector - Combat
      children [
       {
        childComposite=&self:BTComposite_Selector_10
        childTask=null
        decorators [
         &self:FortBTDecorator_GameplayAbility_HasGameplayAbility_0
         &self:FortBTDecorator_GameplayAbility_CanActivate_5
         &self:FortBTDecorator_GameplayAbility_CompareDistance_0
        ]
       }
       {
        childComposite=&self:BTComposite_Selector_9
        childTask=null
        decorators [
         &self:FortBTDecorator_GameplayAbility_HasGameplayAbility_3
         &self:FortBTDecorator_GameplayAbility_CanActivate_10
         &self:FortBTDecorator_GameplayAbility_CompareDistance_7
        ]
       }
       {
        childComposite=&self:BTComposite_Selector_19
        childTask=null
        decorators [
         &self:FortBTDecorator_GameplayAbility_HasGameplayAbility_8
         &self:BTDecorator_GameplayTagQuery_12
        ]
       }
       {
        childComposite=&self:BTComposite_Selector_12
        childTask=null
        decorators [
         &self:FortBTDecorator_GameplayAbility_HasGameplayAbility_9
         &self:FortBTDecorator_GameplayAbility_CanActivate_14
         &self:FortBTDecorator_GameplayAbility_CompareDistance_20
         &self:BTDecorator_GameplayTagQuery_10
        ]
       }
       {
        childComposite=&self:BTComposite_Selector_1
        childTask=null
        decorators [
         &self:FortBTDecorator_GameplayAbility_HasGameplayAbility_12
         &self:FortBTDecorator_GameplayAbility_CanActivate_11
         &self:FortBTDecorator_GameplayAbility_CompareDistance_21
         &self:BTDecorator_GameplayTagQuery_9
        ]
       }
       {
        childComposite=&self:BTComposite_Selector_2
        childTask=null
        decorators [
         &self:FortBTDecorator_GameplayAbility_HasGameplayAbility_7
         &self:FortBTDecorator_GameplayAbility_CanActivate_7
         &self:FortBTDecorator_GameplayAbility_CompareDistance_15
         &self:BTDecorator_GameplayTagQuery_1
        ]
       }
      ]
      services []
     }
     services []
    }
    Sequence {
     class=BTComposite_Sequence
     children [
      Wait {class=BTTask_Wait, properties {waitTime {defaultValue=1}}}
      Selector {
       class=BTComposite_Selector
       children [
        GameplayAbility {
         class=FortBTTask_ExecuteGameplayAbility
         decorators [
          Has Gameplay Ability {
           class=FortBTDecorator_GameplayAbility_HasGameplayAbility
           properties {gameplayAbilityTagContainer {defaultValue {gameplayTags [{tagName=Athena.Faction.Monster.Caretaker.Ability.NonAttack.WanderNearGoalOutside}]}}}
          }
          Can Activate Gameplay Ability {
           class=FortBTDecorator_GameplayAbility_CanActivate
           properties {
            flowAbortMode=None
            gameplayAbilityTagContainer {defaultValue {gameplayTags [{tagName=Athena.Faction.Monster.Caretaker.Ability.NonAttack.WanderNearGoalOutside}]}}
            targetValue {key=GoalActor}
           }
          }
          Doesn't have Inside Sideways {
           class=BTDecorator_GameplayTagQuery
           properties {
            flowAbortMode=None
            nodeName=Doesn't have Inside Sideways
            gameplayTagQuery {
             autoDescription=" NONE( Granted.Athena.Sideways.InsideSideways )"
             queryTokenStream [0,1,3,1,0]
             tagDictionary [{tagName=Granted.Athena.Sideways.InsideSideways}]
            }
           }
          }
         ]
         properties {
          nodeName=GameplayAbility
          gameplayAbilityTagContainer {defaultValue {gameplayTags [{tagName=Athena.Faction.Monster.Caretaker.Ability.NonAttack.WanderNearGoalOutside}]}}
         }
        }
        CaretakerMoveTo {
         class=FortBTTask_CaretakerMoveTo
         children []
         decorators [
          Has Inside Sideways {
           class=BTDecorator_GameplayTagQuery
           properties {
            flowAbortMode=None
            nodeName=Has Inside Sideways
            gameplayTagQuery {
             autoDescription=" ALL( Granted.Athena.Sideways.InsideSideways )"
             queryTokenStream [0,1,2,1,0]
             tagDictionary [{tagName=Granted.Athena.Sideways.InsideSideways}]
            }
           }
          }
          NOT Within Primary Melee Approach Range? {
           class=FortBTDecorator_GameplayAbility_CompareDistance
           properties {
            bInverseCondition=false
            flowAbortMode=None
            nodeName=NOT Within Primary Melee Approach Range?
            activeAbilityTagsToSkipTestingValue {key=None, defaultValue {gameplayTags []}}
            distanceComparisons [{
              addedOffset=0
              bUseAddedOffset=false
              bUseOverriddenValue=false
              comparisonType=ThreeDimensions
              operator=Greater
              overriddenValue=0
              distanceDataTags {gameplayTags [{tagName=NPC.Ability.MaxApproachRange}]}
             }]
            gameplayAbilityTagContainer {key=None, defaultValue {gameplayTags [{tagName=NPC.Ability.Attack.Melee.Primary}]}}
            targetValue {baseClass=&Actor, defaultValue=null, key=GoalActor}
           }
          }
          Blackboard Based Condition {
           class=BTDecorator_Blackboard
           properties {
            arithmeticOperation=Equal
            flowAbortMode=None
            intValue=9
            nodeName=Blackboard Based Condition
            stringValue=EBotMovementState::Success
            blackboardKey {selectedKeyName=AIEvaluator_PathExists}
           }
          }
          Blackboard Based Condition {
           class=BTDecorator_Blackboard
           properties {
            arithmeticOperation=Equal
            flowAbortMode=None
            intValue=10
            nodeName=Blackboard Based Condition
            stringValue=EBotMovementState::Success_Partial
            blackboardKey {selectedKeyName=AIEvaluator_PathExists}
           }
          }
         ]
         properties {
          bEnableSlowdownAtGoal=false
          bFinishMoveOnOverlap=true
          bIgnoreRestartSelf=false
          bMoveDirectlyTowards=false
          bStopAtGoal=true
          nodeName=CaretakerMoveTo
          pathObstacleAction=Ignore
          acceptableRadius {defaultValue=750, key=None}
          bAllowPartialPath {defaultValue=true, key=None}
          bAllowStrafe {defaultValue=false, key=None}
          bProjectGoalLocation {defaultValue=true, key=None}
          bReachTestIncludesAgentRadius {defaultValue=true, key=None}
          bReachTestIncludesGoalRadius {defaultValue=true, key=None}
          bRequireNavigableEndLocation {defaultValue=true, key=None}
          bStartFromPreviousPath {defaultValue=false, key=None}
          bTrackMovingGoal {defaultValue=true, key=None}
          blackboardKey {bNoneIsAllowedValue=false, selectedKeyName=GoalActor}
          filterClass {
           baseClass=&NavigationQueryFilter
           defaultValue=&NavFilter_Caretaker_NoBlockers_C
           key=None
          }
          focalPointWhileMoving {bNoneIsAllowedValue=true, selectedKeyName=None}
          observedBlackboardValueTolerance {defaultValue=4.75, key=None}
         }
         services []
        }
        GameplayAbility {
         class=FortBTTask_ExecuteGameplayAbility
         decorators [
          Can Activate Gameplay Ability {
           class=FortBTDecorator_GameplayAbility_CanActivate
           properties {
            flowAbortMode=None
            gameplayAbilityTagContainer {defaultValue {gameplayTags [{tagName=NPC.Ability.NonAttack.PlayMontage.Wander.NearGoal}]}}
            targetValue {key=GoalActor}
           }
          }
          Has Inside Sideways {
           class=BTDecorator_GameplayTagQuery
           properties {
            flowAbortMode=None
            nodeName=Has Inside Sideways
            gameplayTagQuery {
             autoDescription=" ALL( Granted.Athena.Sideways.InsideSideways )"
             queryTokenStream [0,1,2,1,0]
             tagDictionary [{tagName=Granted.Athena.Sideways.InsideSideways}]
            }
           }
          }
          Has Gameplay Ability {
           class=FortBTDecorator_GameplayAbility_HasGameplayAbility
           properties {gameplayAbilityTagContainer {defaultValue {gameplayTags [{tagName=NPC.Ability.NonAttack.PlayMontage.Wander.NearGoal}]}}}
          }
         ]
         properties {
          nodeName=GameplayAbility
          gameplayAbilityTagContainer {defaultValue {gameplayTags [{tagName=NPC.Ability.NonAttack.PlayMontage.Wander.NearGoal}]}}
         }
        }
        GameplayAbility {
         class=FortBTTask_ExecuteGameplayAbility
         decorators [
          Has Gameplay Ability {
           class=FortBTDecorator_GameplayAbility_HasGameplayAbility
           properties {gameplayAbilityTagContainer {defaultValue {gameplayTags [{tagName=NPC.Ability.NonAttack.Wander.NearGoal}]}}}
          }
          Has Inside Sideways {
           class=BTDecorator_GameplayTagQuery
           properties {
            flowAbortMode=None
            nodeName=Has Inside Sideways
            gameplayTagQuery {
             autoDescription=" ALL( Granted.Athena.Sideways.InsideSideways )"
             queryTokenStream [0,1,2,1,0]
             tagDictionary [{tagName=Granted.Athena.Sideways.InsideSideways}]
            }
           }
          }
          Can Activate Gameplay Ability {
           class=FortBTDecorator_GameplayAbility_CanActivate
           properties {
            flowAbortMode=None
            gameplayAbilityTagContainer {defaultValue {gameplayTags [{tagName=NPC.Ability.NonAttack.Wander.NearGoal}]}}
            targetValue {key=GoalActor}
           }
          }
         ]
         properties {nodeName=GameplayAbility, gameplayAbilityTagContainer {defaultValue {gameplayTags [{tagName=NPC.Ability.NonAttack.Wander.NearGoal}]}}}
        }
       ]
       decorators []
       properties {
        children [
         {
          childComposite=null
          childTask=&self:FortBTTask_ExecuteGameplayAbility_8
          decorators [
           &self:FortBTDecorator_GameplayAbility_HasGameplayAbility_2
           &self:FortBTDecorator_GameplayAbility_CanActivate_13
           &self:BTDecorator_GameplayTagQuery_7
          ]
         }
         {
          childComposite=null
          childTask=&self:FortBTTask_CaretakerMoveTo_2
          decoratorOps [{number=3, operation=And}, {number=0}, {number=1}, {number=2, operation=Or}, {number=2}, {number=3}]
          decorators [
           &self:BTDecorator_GameplayTagQuery_4
           &self:FortBTDecorator_GameplayAbility_CompareDistance_2
           &self:BTDecorator_Blackboard_22
           &self:BTDecorator_Blackboard_23
          ]
         }
         {
          childComposite=null
          childTask=&self:FortBTTask_ExecuteGameplayAbility_21
          decorators [
           &self:FortBTDecorator_GameplayAbility_CanActivate_6
           &self:BTDecorator_GameplayTagQuery_3
           &self:FortBTDecorator_GameplayAbility_HasGameplayAbility_4
          ]
         }
         {
          childComposite=null
          childTask=&self:FortBTTask_ExecuteGameplayAbility_11
          decorators [
           &self:FortBTDecorator_GameplayAbility_HasGameplayAbility_1
           &self:BTDecorator_GameplayTagQuery_2
           &self:FortBTDecorator_GameplayAbility_CanActivate_1
          ]
         }
        ]
        services []
       }
       services []
      }
     ]
     decorators [Not Escaping Into Storm {
       class=BTDecorator_GameplayTagQuery
       properties {
        flowAbortMode=None
        nodeName=Not Escaping Into Storm
        gameplayTagQuery {
         autoDescription=" NONE( NPC.Ability.NonAttack.Caretaker.EscapeIntoStorm )"
         queryTokenStream [0,1,3,1,0]
         tagDictionary [{tagName=NPC.Ability.NonAttack.Caretaker.EscapeIntoStorm}]
        }
       }
      }]
     properties {children [
       {childTask=&self:BTTask_Wait_16}
       {childComposite=&self:BTComposite_Selector_3, childTask=null}
      ]}
    }
   ]
   decorators [
    Has GoalActor? {
     class=BTDecorator_Blackboard
     properties {
      arithmeticOperation=Equal
      intValue=0
      nodeName=Has GoalActor?
      notifyObserver=ValueChange
      stringValue=""
      blackboardKey {selectedKeyName=GoalActor}
     }
    }
    Is Goal Pawn {
     class=FortBTDecorator_IsGoalPawn
     properties {bInverseCondition=false, flowAbortMode=Both, nodeName=Is Goal Pawn, blackboardKey {bNoneIsAllowedValue=false, selectedKeyName=GoalActor}}
    }
   ]
   properties {
    children [
     {
      childComposite=&self:BTComposite_Selector_8
      childTask=null
      decoratorOps [{number=5, operation=And}, {number=0}, {number=1}, {number=2}, {number=3}, {number=2, operation=And}, {number=4}, {number=5}]
      decorators [
       &self:HasActiveAbilityWithAnyMatchingTag_C_0
       &self:HasActiveAbilityWithAnyMatchingTag_C_1
       &self:HasActiveAbilityWithAnyMatchingTag_C_2
       &self:HasActiveAbilityWithAnyMatchingTag_C_3
       &self:BTDecorator_Blackboard_5
       &self:BTDecorator_Blackboard_6
      ]
     }
     {childComposite=&self:BTComposite_Selector_21, childTask=null, decorators []}
     {
      childComposite=&self:BTComposite_Sequence_20
      childTask=null
      decorators [&self:BTDecorator_GameplayTagQuery_11]
     }
    ]
    services [&self:FortAthenaBTService_AIEvaluator_3]
   }
   services [Evaluating...Path Exists {
     class=FortAthenaBTService_AIEvaluator
     properties {
      bCallTickOnSearchStart=true
      bRestartTimerOnEachActivation=true
      interval=0.75
      nodeName=Evaluating...Path Exists
      randomDeviation=0.25
      aIEvaluatorClass {defaultValue=/Caretaker/AIEvaluator/Caretaker_AIEval_PathExists.Caretaker_AIEval_PathExists_C}
     }
    }]
  }
  Selector {
   class=BTComposite_Selector
   children [
    Selector {
     class=BTComposite_Selector
     children [
      GameplayAbility {
       class=FortBTTask_ExecuteGameplayAbility
       decorators [
        Can Activate Gameplay Ability {
         class=FortBTDecorator_GameplayAbility_CanActivate
         properties {
          flowAbortMode=None
          gameplayAbilityTagContainer {defaultValue {gameplayTags [{tagName=NPC.Ability.NonAttack.Wander.Escape}]}}
          targetValue {key=GoalActor}
         }
        }
        Has NO GoalActor? {
         class=BTDecorator_Blackboard
         properties {
          arithmeticOperation=Equal
          basicOperation=NotSet
          intValue=0
          nodeName=Has NO GoalActor?
          stringValue=""
          blackboardKey {selectedKeyName=GoalActor}
         }
        }
       ]
       properties {nodeName=GameplayAbility, gameplayAbilityTagContainer {defaultValue {gameplayTags [{tagName=NPC.Ability.NonAttack.Wander.Escape}]}}}
      }
      GameplayAbility {
       class=FortBTTask_ExecuteGameplayAbility
       decorators [
        Can Activate Gameplay Ability {
         class=FortBTDecorator_GameplayAbility_CanActivate
         properties {
          flowAbortMode=None
          gameplayAbilityTagContainer {defaultValue {gameplayTags [{tagName=NPC.Ability.NonAttack.Wander.Escape}]}}
          targetValue {key=GoalActor}
         }
        }
        Has GoalActor? {
         class=BTDecorator_Blackboard
         properties {
          arithmeticOperation=Equal
          intValue=0
          nodeName=Has GoalActor?
          notifyObserver=ValueChange
          stringValue=""
          blackboardKey {selectedKeyName=GoalActor}
         }
        }
       ]
       properties {nodeName=GameplayAbility, gameplayAbilityTagContainer {defaultValue {gameplayTags [{tagName=NPC.Ability.NonAttack.Wander.Escape}]}}}
      }
      Selector {
       class=BTComposite_Selector
       children [Wait {class=BTTask_Wait, properties {waitTime {defaultValue=1}}}]
       decorators [Is On Cooldown {
         class=FortBTDecorator_GameplayAbility_IsOnCooldown
         properties {
          bInverseCondition=false
          flowAbortMode=None
          nodeName=Is On Cooldown
          activeAbilityTagsToSkipTestingValue {key=None, defaultValue {gameplayTags []}}
          gameplayAbilityTagContainer {key=None, defaultValue {gameplayTags [{tagName=NPC.Ability.NonAttack.Caretaker.EscapeIntoStorm}]}}
          targetValue {baseClass=&Actor, defaultValue=null, key=GoalActor}
         }
        }]
       properties {children [{childComposite=null, childTask=&self:BTTask_Wait_20, decorators []}], services []}
       services []
      }
     ]
     decorators [
      Has Gameplay Ability {
       class=FortBTDecorator_GameplayAbility_HasGameplayAbility
       properties {gameplayAbilityTagContainer {defaultValue {gameplayTags [{tagName=NPC.Ability.NonAttack.Wander.Escape}]}}}
      }
      Escaping Into Storm {
       class=BTDecorator_GameplayTagQuery
       properties {
        flowAbortMode=None
        nodeName=Escaping Into Storm
        gameplayTagQuery {
         autoDescription=" ANY( NPC.Ability.NonAttack.Caretaker.EscapeIntoStorm )"
         queryTokenStream [0,1,1,1,0]
         tagDictionary [{tagName=NPC.Ability.NonAttack.Caretaker.EscapeIntoStorm}]
        }
       }
      }
     ]
     properties {
      children [
       {
        childComposite=null
        childTask=&self:FortBTTask_ExecuteGameplayAbility_16
        decorators [
         &self:FortBTDecorator_GameplayAbility_CanActivate_0
         &self:BTDecorator_Blackboard_0
        ]
       }
       {
        childComposite=null
        childTask=&self:FortBTTask_ExecuteGameplayAbility_17
        decorators [
         &self:FortBTDecorator_GameplayAbility_CanActivate_8
         &self:BTDecorator_Blackboard_9
        ]
       }
       {
        childComposite=&self:BTComposite_Selector_23
        childTask=null
        decorators [&self:FortBTDecorator_GameplayAbility_IsOnCooldown_0]
       }
      ]
      services []
     }
     services []
    }
    Selector {
     class=BTComposite_Selector
     children [
      GameplayAbility {
       class=FortBTTask_ExecuteGameplayAbility
       decorators [
        Can Activate Gameplay Ability {
         class=FortBTDecorator_GameplayAbility_CanActivate
         properties {
          flowAbortMode=None
          gameplayAbilityTagContainer {defaultValue {gameplayTags [{tagName=NPC.Ability.NonAttack.Wander.Freely}]}}
          targetValue {key=GoalActor}
         }
        }
        Has NO GoalActor? {
         class=BTDecorator_Blackboard
         properties {
          arithmeticOperation=Equal
          basicOperation=NotSet
          intValue=0
          nodeName=Has NO GoalActor?
          stringValue=""
          blackboardKey {selectedKeyName=GoalActor}
         }
        }
       ]
       properties {nodeName=GameplayAbility, gameplayAbilityTagContainer {defaultValue {gameplayTags [{tagName=NPC.Ability.NonAttack.Wander.Freely}]}}}
      }
      GameplayAbility {
       class=FortBTTask_ExecuteGameplayAbility
       decorators [
        Can Activate Gameplay Ability {
         class=FortBTDecorator_GameplayAbility_CanActivate
         properties {
          flowAbortMode=None
          gameplayAbilityTagContainer {defaultValue {gameplayTags [{tagName=NPC.Ability.NonAttack.Wander.Freely}]}}
          targetValue {key=None}
         }
        }
        Has GoalActor? {
         class=BTDecorator_Blackboard
         properties {
          arithmeticOperation=Equal
          intValue=0
          nodeName=Has GoalActor?
          notifyObserver=ValueChange
          stringValue=""
          blackboardKey {selectedKeyName=GoalActor}
         }
        }
       ]
       properties {nodeName=GameplayAbility, gameplayAbilityTagContainer {defaultValue {gameplayTags [{tagName=NPC.Ability.NonAttack.Wander.Freely}]}}}
      }
      Selector {
       class=BTComposite_Selector
       children [Wait {class=BTTask_Wait, properties {randomDeviation {defaultValue=2}, waitTime {defaultValue=3}}}]
       decorators [Is On Cooldown {
         class=FortBTDecorator_GameplayAbility_IsOnCooldown
         properties {
          bInverseCondition=false
          flowAbortMode=None
          nodeName=Is On Cooldown
          activeAbilityTagsToSkipTestingValue {key=None, defaultValue {gameplayTags []}}
          gameplayAbilityTagContainer {key=None, defaultValue {gameplayTags [{tagName=NPC.Ability.NonAttack.Wander.Freely}]}}
          targetValue {baseClass=&Actor, defaultValue=null, key=None}
         }
        }]
       properties {children [{childComposite=null, childTask=&self:BTTask_Wait_15, decorators []}], services []}
       services []
      }
     ]
     decorators [
      Has Gameplay Ability {
       class=FortBTDecorator_GameplayAbility_HasGameplayAbility
       properties {gameplayAbilityTagContainer {defaultValue {gameplayTags [{tagName=NPC.Ability.NonAttack.Wander.Freely}]}}}
      }
      Not Escaping Into Storm {
       class=BTDecorator_GameplayTagQuery
       properties {
        flowAbortMode=None
        nodeName=Not Escaping Into Storm
        gameplayTagQuery {
         autoDescription=" NONE( NPC.Ability.NonAttack.Caretaker.EscapeIntoStorm )"
         queryTokenStream [0,1,3,1,0]
         tagDictionary [{tagName=NPC.Ability.NonAttack.Caretaker.EscapeIntoStorm}]
        }
       }
      }
     ]
     properties {
      children [
       {
        childComposite=null
        childTask=&self:FortBTTask_ExecuteGameplayAbility_0
        decorators [
         &self:FortBTDecorator_GameplayAbility_CanActivate_2
         &self:BTDecorator_Blackboard_4
        ]
       }
       {
        childComposite=null
        childTask=&self:FortBTTask_ExecuteGameplayAbility_3
        decorators [
         &self:FortBTDecorator_GameplayAbility_CanActivate_3
         &self:BTDecorator_Blackboard_7
        ]
       }
       {
        childComposite=&self:BTComposite_Selector_17
        childTask=null
        decorators [&self:FortBTDecorator_GameplayAbility_IsOnCooldown_1]
       }
      ]
      services []
     }
     services []
    }
   ]
   decorators []
   properties {
    children [
     {
      childComposite=&self:BTComposite_Selector_22
      childTask=null
      decorators [
       &self:FortBTDecorator_GameplayAbility_HasGameplayAbility_6
       &self:BTDecorator_GameplayTagQuery_13
      ]
     }
     {
      childComposite=&self:BTComposite_Selector_28
      childTask=null
      decorators [
       &self:FortBTDecorator_GameplayAbility_HasGameplayAbility_15
       &self:BTDecorator_GameplayTagQuery_0
      ]
     }
    ]
    services []
   }
   services []
  }
 ]
 decorators []
 properties {
  children [
   {
    childComposite=&self:BTComposite_Selector_5
    childTask=null
    decorators [&self:BTDecorator_Blackboard_3, &self:FortBTDecorator_IsGoalPawn_0]
   }
   {childComposite=&self:BTComposite_Selector_24, childTask=null, decorators []}
  ]
  services []
 }
 services []
}
